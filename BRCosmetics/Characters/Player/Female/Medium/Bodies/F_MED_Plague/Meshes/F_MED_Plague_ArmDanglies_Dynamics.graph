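# Physics asset for the dangling forearm tassels on the F_MED_Plague female player cosmetic body.
# Each arm has six two-bone capsule chains, in up, mid and out pairs.
# The pairs hang from lowerarm, lowerarm_twist_02 and lowerarm_twist_01 respectively, with wide swing and narrow twist limits.
# Capsules on the forearm bones serve as anchors, and both twist bones join lowerarm with free angular motion.
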PhysicsAsset /BRCosmetics/Characters/Player/Female/Medium/Bodies/F_MED_Plague/Meshes/F_MED_Plague_ArmDanglies_Dynamics
bodies {
 dyn_forearm_tassle_mid_A_01_l [{shapeType=Capsule, length=2.9075, radius=0.5454, center=<0 -2.0585 0>, rotation=<rot 0 0 90>}]
 dyn_forearm_tassle_mid_A_01_r [{shapeType=Capsule, length=2.9075, radius=0.5454, center=<0 2.0585 0>, rotation=<rot 0 0 90>}]
 dyn_forearm_tassle_mid_A_02_l [{shapeType=Capsule, length=2.9075, radius=0.5454, center=<0 -2.0585 0>, rotation=<rot 0 0 90>}]
 dyn_forearm_tassle_mid_A_02_r [{shapeType=Capsule, length=2.9075, radius=0.5454, center=<0 2.0585 0>, rotation=<rot 0 0 90>}]
 dyn_forearm_tassle_mid_B_01_l [{shapeType=Capsule, length=2.9075, radius=0.5454, center=<0 -2.0585 0>, rotation=<rot 0 0 90>}]
 dyn_forearm_tassle_mid_B_01_r [{shapeType=Capsule, length=2.9075, radius=0.5454, center=<0 2.0585 0>, rotation=<rot 0 0 90>}]
 dyn_forearm_tassle_mid_B_02_l [{shapeType=Capsule, length=2.9075, radius=0.5454, center=<0 -2.0585 0>, rotation=<rot 0 0 90>}]
 dyn_forearm_tassle_mid_B_02_r [{shapeType=Capsule, length=2.9075, radius=0.5454, center=<0 2.0585 0>, rotation=<rot 0 0 90>}]
 dyn_forearm_tassle_out_A_01_l [{shapeType=Capsule, length=2.9075, radius=0.5454, center=<0 -2.0585 0>, rotation=<rot 0 0 90>}]
 dyn_forearm_tassle_out_A_01_r [{shapeType=Capsule, length=2.9075, radius=0.5454, center=<0 2.0585 0>, rotation=<rot 0 0 90>}]
 dyn_forearm_tassle_out_A_02_l [{shapeType=Capsule, length=2.9075, radius=0.5454, center=<0 -2.0585 0>, rotation=<rot 0 0 90>}]
 dyn_forearm_tassle_out_A_02_r [{shapeType=Capsule, length=2.9075, radius=0.5454, center=<0 2.0585 0>, rotation=<rot 0 0 90>}]
 dyn_forearm_tassle_out_B_01_l [{shapeType=Capsule, length=2.9075, radius=0.5454, center=<0 -2.0585 0>, rotation=<rot 0 0 90>}]
 dyn_forearm_tassle_out_B_01_r [{shapeType=Capsule, length=2.9075, radius=0.5454, center=<0 2.0585 0>, rotation=<rot 0 0 90>}]
 dyn_forearm_tassle_out_B_02_l [{shapeType=Capsule, length=2.9075, radius=0.5454, center=<0 -2.0585 0>, rotation=<rot 0 0 90>}]
 dyn_forearm_tassle_out_B_02_r [{shapeType=Capsule, length=2.9075, radius=0.5454, center=<0 2.0585 0>, rotation=<rot 0 0 90>}]
 dyn_forearm_tassle_up_A_01_l [{shapeType=Capsule, length=2.9075, radius=0.5454, center=<0 -2.0585 0>, rotation=<rot 0 0 90>}]
 dyn_forearm_tassle_up_A_01_r [{shapeType=Capsule, length=2.9075, radius=0.5454, center=<0 2.0585 0>, rotation=<rot 0 0 90>}]
 dyn_forearm_tassle_up_A_02_l [{shapeType=Capsule, length=2.9075, radius=0.5454, center=<0 -2.0585 0>, rotation=<rot 0 0 90>}]
 dyn_forearm_tassle_up_A_02_r [{shapeType=Capsule, length=2.9075, radius=0.5454, center=<0 2.0585 0>, rotation=<rot 0 0 90>}]
 dyn_forearm_tassle_up_B_01_l [{shapeType=Capsule, length=2.9075, radius=0.5454, center=<0 -2.0585 0>, rotation=<rot 0 0 90>}]
 dyn_forearm_tassle_up_B_01_r [{shapeType=Capsule, length=2.9075, radius=0.5454, center=<0 2.0585 0>, rotation=<rot 0 0 90>}]
 dyn_forearm_tassle_up_B_02_l [{shapeType=Capsule, length=2.9075, radius=0.5454, center=<0 -2.0585 0>, rotation=<rot 0 0 90>}]
 dyn_forearm_tassle_up_B_02_r [{shapeType=Capsule, length=2.9075, radius=0.5454, center=<0 2.0585 0>, rotation=<rot 0 0 90>}]
 lowerarm_l [{shapeType=Capsule, length=2, radius=2}]
 lowerarm_r [{shapeType=Capsule, length=2, radius=2}]
 lowerarm_twist_01_l [{shapeType=Capsule, length=2, radius=2}]
 lowerarm_twist_01_r [{shapeType=Capsule, length=2, radius=2}]
 lowerarm_twist_02_l [{shapeType=Capsule, length=2, radius=2}]
 lowerarm_twist_02_r [{shapeType=Capsule, length=2, radius=2}]
}
constraints [
 {bone1Name=dyn_forearm_tassle_out_A_01_l, bone2Name=lowerarm_twist_01_l, swing1LimitDegrees=90, swing2LimitDegrees=90, twistLimitDegrees=10}
 {
  bone1Name=dyn_forearm_tassle_out_A_02_l
  bone2Name=dyn_forearm_tassle_out_A_01_l
  swing1LimitDegrees=90
  swing2LimitDegrees=90
  twistLimitDegrees=10
 }
 {bone1Name=dyn_forearm_tassle_out_B_01_l, bone2Name=lowerarm_twist_01_l, swing1LimitDegrees=90, swing2LimitDegrees=90, twistLimitDegrees=10}
 {
  bone1Name=dyn_forearm_tassle_out_B_02_l
  bone2Name=dyn_forearm_tassle_out_B_01_l
  swing1LimitDegrees=90
  swing2LimitDegrees=90
  twistLimitDegrees=10
 }
 {bone1Name=dyn_forearm_tassle_mid_B_01_l, bone2Name=lowerarm_twist_02_l, swing1LimitDegrees=90, swing2LimitDegrees=90, twistLimitDegrees=10}
 {
  bone1Name=dyn_forearm_tassle_mid_B_02_l
  bone2Name=dyn_forearm_tassle_mid_B_01_l
  swing1LimitDegrees=90
  swing2LimitDegrees=90
  twistLimitDegrees=10
 }
 {bone1Name=dyn_forearm_tassle_mid_A_01_l, bone2Name=lowerarm_twist_02_l, swing1LimitDegrees=90, swing2LimitDegrees=90, twistLimitDegrees=10}
 {
  bone1Name=dyn_forearm_tassle_mid_A_02_l
  bone2Name=dyn_forearm_tassle_mid_A_01_l
  swing1LimitDegrees=90
  swing2LimitDegrees=90
  twistLimitDegrees=10
 }
 {
  bone1Name=dyn_forearm_tassle_up_A_02_l
  bone2Name=dyn_forearm_tassle_up_A_01_l
  swing1LimitDegrees=90
  swing2LimitDegrees=90
  twistLimitDegrees=10
 }
 {
  bone1Name=dyn_forearm_tassle_up_B_02_l
  bone2Name=dyn_forearm_tassle_up_B_01_l
  swing1LimitDegrees=90
  swing2LimitDegrees=90
  twistLimitDegrees=10
 }
 {bone1Name=dyn_forearm_tassle_out_A_01_r, bone2Name=lowerarm_twist_01_r, swing1LimitDegrees=90, swing2LimitDegrees=90, twistLimitDegrees=10}
 {
  bone1Name=dyn_forearm_tassle_out_A_02_r
  bone2Name=dyn_forearm_tassle_out_A_01_r
  swing1LimitDegrees=90
  swing2LimitDegrees=90
  twistLimitDegrees=10
 }
 {bone1Name=dyn_forearm_tassle_out_B_01_r, bone2Name=lowerarm_twist_01_r, swing1LimitDegrees=90, swing2LimitDegrees=90, twistLimitDegrees=10}
 {
  bone1Name=dyn_forearm_tassle_out_B_02_r
  bone2Name=dyn_forearm_tassle_out_B_01_r
  swing1LimitDegrees=90
  swing2LimitDegrees=90
  twistLimitDegrees=10
 }
 {bone1Name=dyn_forearm_tassle_mid_B_01_r, bone2Name=lowerarm_twist_02_r, swing1LimitDegrees=90, swing2LimitDegrees=90, twistLimitDegrees=10}
 {
  bone1Name=dyn_forearm_tassle_mid_B_02_r
  bone2Name=dyn_forearm_tassle_mid_B_01_r
  swing1LimitDegrees=90
  swing2LimitDegrees=90
  twistLimitDegrees=10
 }
 {bone1Name=dyn_forearm_tassle_mid_A_01_r, bone2Name=lowerarm_twist_02_r, swing1LimitDegrees=90, swing2LimitDegrees=90, twistLimitDegrees=10}
 {
  bone1Name=dyn_forearm_tassle_mid_A_02_r
  bone2Name=dyn_forearm_tassle_mid_A_01_r
  swing1LimitDegrees=90
  swing2LimitDegrees=90
  twistLimitDegrees=10
 }
 {
  bone1Name=dyn_forearm_tassle_up_A_02_r
  bone2Name=dyn_forearm_tassle_up_A_01_r
  swing1LimitDegrees=90
  swing2LimitDegrees=90
  twistLimitDegrees=10
 }
 {
  bone1Name=dyn_forearm_tassle_up_B_02_r
  bone2Name=dyn_forearm_tassle_up_B_01_r
  swing1LimitDegrees=90
  swing2LimitDegrees=90
  twistLimitDegrees=10
 }
 {
  bone1Name=lowerarm_twist_01_l
  bone2Name=lowerarm_l
  swing1LimitDegrees=45
  swing1Motion=Free
  swing2LimitDegrees=45
  swing2Motion=Free
  twistLimitDegrees=45
  twistMotion=Free
 }
 {
  bone1Name=lowerarm_twist_02_l
  bone2Name=lowerarm_l
  swing1LimitDegrees=45
  swing1Motion=Free
  swing2LimitDegrees=45
  swing2Motion=Free
  twistLimitDegrees=45
  twistMotion=Free
 }
 {bone1Name=dyn_forearm_tassle_up_A_01_l, bone2Name=lowerarm_l, swing1LimitDegrees=90, swing2LimitDegrees=90, twistLimitDegrees=10}
 {bone1Name=dyn_forearm_tassle_up_B_01_l, bone2Name=lowerarm_l, swing1LimitDegrees=90, swing2LimitDegrees=90, twistLimitDegrees=10}
 {
  bone1Name=lowerarm_twist_01_r
  bone2Name=lowerarm_r
  swing1LimitDegrees=45
  swing1Motion=Free
  swing2LimitDegrees=45
  swing2Motion=Free
  twistLimitDegrees=45
  twistMotion=Free
 }
 {
  bone1Name=lowerarm_twist_02_r
  bone2Name=lowerarm_r
  swing1LimitDegrees=45
  swing1Motion=Free
  swing2LimitDegrees=45
  swing2Motion=Free
  twistLimitDegrees=45
  twistMotion=Free
 }
 {bone1Name=dyn_forearm_tassle_up_A_01_r, bone2Name=lowerarm_r, swing1LimitDegrees=90, swing2LimitDegrees=90, twistLimitDegrees=10}
 {bone1Name=dyn_forearm_tassle_up_B_01_r, bone2Name=lowerarm_r, swing1LimitDegrees=90, swing2LimitDegrees=90, twistLimitDegrees=10}
]
