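# Physics asset that drives skirt dynamics for the female medium Plague body cosmetic.
# Pelvis and leg bodies accompany bone chains hung from the pelvis around the front, sides and back, using mostly capsules behind and boxes elsewhere.
# Chain joints have tight twist limits, with the widest swing on front chains.
# Angularly free cross-constraints tie some neighboring chains into a partial lattice, with extra rear-diagonal links and some bulkier boxes on the right.
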PhysicsAsset /BRCosmetics/Characters/Player/Female/Medium/Bodies/F_MED_Plague/Meshes/F_MED_Plague_Dynamics
bodies {
 calf_l [{shapeType=Capsule, length=47.8666, radius=6.8286, center=<-26.763 -1.579 -0.3374>, rotation=<rot 0.1568 86.7612 -88.2069>}]
 calf_r [{shapeType=Capsule, length=47.8666, radius=6.8286, center=<26.763 1.579 0.3374>, rotation=<rot 0.1568 86.7612 -88.2069>}]
 dyn_skirt_bk_01 [{shapeType=Capsule, length=9.5312, radius=6.9688, center=<6.5504 0 6>, rotation=<rot 90 0 0>}]
 dyn_skirt_bk_01_l [{shapeType=Capsule, length=9.5312, radius=6.9688, center=<6.5504 0 6>, rotation=<rot 90 0 0>}]
 dyn_skirt_bk_01_r [{shapeType=Capsule, length=9.5312, radius=6.9688, center=<-6.5504 0 -6>, rotation=<rot 90 0 0>}]
 dyn_skirt_bk_01a [{shapeType=Capsule, length=9.5312, radius=6.9688, center=<6.5504 0 6>, rotation=<rot 90 0 0>}]
 dyn_skirt_bk_01a_l [{shapeType=Capsule, length=9.5312, radius=6.9688, center=<6.5504 0 6>, rotation=<rot 90 0 0>}]
 dyn_skirt_bk_01a_r [{shapeType=Capsule, length=9.5312, radius=6.9688, center=<-6.5504 0 -6>, rotation=<rot 90 0 0>}]
 dyn_skirt_bk_02 [{shapeType=Capsule, length=11.0625, radius=11.0312, center=<6.5504 -0.0002 10>, rotation=<rot 90 0 0>}]
 dyn_skirt_bk_02_l [{shapeType=Capsule, length=11.0625, radius=11.0312, center=<6.5504 -0.0002 10>, rotation=<rot 90 0 0>}]
 dyn_skirt_bk_02_r [{shapeType=Capsule, length=11.0625, radius=11.0312, center=<-6.5504 -0.0002 -10>, rotation=<rot 90 0 0>}]
 dyn_skirt_bk_02a [{shapeType=Capsule, length=11.0625, radius=11.0312, center=<6.5504 -0.0002 10>, rotation=<rot 90 0 0>}]
 dyn_skirt_bk_02a_l [{shapeType=Capsule, length=11.0625, radius=11.0312, center=<6.5504 -0.0002 10>, rotation=<rot 90 0 0>}]
 dyn_skirt_bk_02a_r [
  {shapeType=Box, extentX=3.5, extentY=22.3041, extentZ=10.8495, center=<-6.6738 0 0>, rotation=<rot 90 0 0>}
  {
   shapeType=Box
   extentX=3.5
   extentY=22.3041
   extentZ=10.8495
   shapeName=dyn_skirt_bk_02a_r_box
   center=<-5.6389 -14.5751 3.8102>
   rotation=<rot 50 89.9999 89.9998>
  }
 ]
 dyn_skirt_bk_03 [{shapeType=Capsule, length=11.0625, radius=11.0312, center=<6.5504 -0.0002 10>, rotation=<rot 90 0 0>}]
 dyn_skirt_bk_03_l [{shapeType=Capsule, length=11.0625, radius=11.0312, center=<6.5504 -0.0002 10>, rotation=<rot 90 0 0>}]
 dyn_skirt_bk_03_r [{shapeType=Box, extentX=4, extentY=24.8041, extentZ=9.0374, center=<-5.0765 0 0>, rotation=<rot 90 0 0>}]
 dyn_skirt_bkdiag_01_l [{shapeType=Capsule, length=9.5312, radius=6.9688, center=<6.5504 0 6>, rotation=<rot 90 0 0>}]
 dyn_skirt_bkdiag_01_r [{shapeType=Capsule, length=9.5312, radius=6.9688, center=<-6.5504 0 -6>, rotation=<rot 90 0 0>}]
 dyn_skirt_bkdiag_01a_l [{shapeType=Capsule, length=9.5312, radius=6.9688, center=<6.5504 0 6>, rotation=<rot 90 0 0>}]
 dyn_skirt_bkdiag_01a_r [{shapeType=Capsule, length=9.5312, radius=6.9688, center=<-6.5504 0 -6>, rotation=<rot 90 0 0>}]
 dyn_skirt_bkdiag_02_l [{shapeType=Capsule, length=11.0625, radius=11.0312, center=<6.5504 -0.0002 10>, rotation=<rot 90 0 0>}]
 dyn_skirt_bkdiag_02_r [{shapeType=Capsule, length=14.0625, radius=8.0312, center=<-6.5504 -0.0002 -5.1239>, rotation=<rot 90 -17.8276 -17.8277>}]
 dyn_skirt_bkdiag_02a_l [{shapeType=Capsule, length=11.0625, radius=11.0312, center=<6.5504 -0.0002 10>, rotation=<rot 90 0 0>}]
 dyn_skirt_bkdiag_02a_r [{shapeType=Box, extentX=3.5, extentY=20.1434, extentZ=14.9342, center=<-6.6738 0 -1.9857>, rotation=<rot 90 0 0>}]
 dyn_skirt_bkdiag_03_l [{shapeType=Capsule, length=11.0625, radius=11.0312, center=<6.5504 -0.0002 10>, rotation=<rot 90 0 0>}]
 dyn_skirt_bkdiag_03_r [{shapeType=Box, extentX=4, extentY=20.0634, extentZ=12.1161, center=<-4.6172 0 -1.6394>, rotation=<rot 90 0 0>}]
 dyn_skirt_bkmid_01_l [{shapeType=Capsule, length=9.5312, radius=6.9688, center=<6.5504 0 6>, rotation=<rot 90 0 0>}]
 dyn_skirt_bkmid_01_r [{shapeType=Capsule, length=9.5312, radius=6.9688, center=<-6.5504 0 -6>, rotation=<rot 90 0 0>}]
 dyn_skirt_bkmid_01a_l [{shapeType=Capsule, length=9.5312, radius=6.9688, center=<6.5504 0 6>, rotation=<rot 90 0 0>}]
 dyn_skirt_bkmid_01a_r [{shapeType=Capsule, length=9.5312, radius=6.9688, center=<-6.5504 0 -6>, rotation=<rot 90 0 0>}]
 dyn_skirt_bkmid_02_l [{shapeType=Capsule, length=11.0625, radius=11.0312, center=<6.5504 -0.0002 10>, rotation=<rot 90 0 0>}]
 dyn_skirt_bkmid_02_r [{shapeType=Capsule, length=11.0625, radius=11.0312, center=<-6.5504 -0.0002 -10>, rotation=<rot 90 0 0>}]
 dyn_skirt_bkmid_02a_l [{shapeType=Capsule, length=11.0625, radius=11.0312, center=<6.5504 -0.0002 10>, rotation=<rot 90 0 0>}]
 dyn_skirt_bkmid_02a_r [{shapeType=Capsule, length=11.0625, radius=11.0312, center=<-6.5504 -0.0002 -10>, rotation=<rot 90 0 0>}]
 dyn_skirt_bkmid_03_l [{shapeType=Capsule, length=11.0625, radius=11.0312, center=<6.5504 -0.0002 10>, rotation=<rot 90 0 0>}]
 dyn_skirt_bkmid_03_r [{shapeType=Capsule, length=11.0625, radius=11.0312, center=<-6.5504 -0.0002 -10>, rotation=<rot 90 0 0>}]
 dyn_skirt_fr_01 [{shapeType=Box, extentX=1.5, extentY=1.2251, extentZ=13, center=<7 0 0>, rotation=<rot 90 0 0>}]
 dyn_skirt_fr_01_l [{shapeType=Box, extentX=1.5, extentY=1.2251, extentZ=13, center=<7 0 0>, rotation=<rot 90 0 0>}]
 dyn_skirt_fr_01_r [{shapeType=Box, extentX=1.5, extentY=1.2251, extentZ=13, center=<-7 0 0>, rotation=<rot 90 0 0>}]
 dyn_skirt_fr_02 [{shapeType=Box, extentX=1.5, extentY=1.2251, extentZ=13, center=<7 0 0>, rotation=<rot 90 0 0>}]
 dyn_skirt_fr_02_l [{shapeType=Box, extentX=1.5, extentY=1.2251, extentZ=13, center=<7 0 0>, rotation=<rot 90 0 0>}]
 dyn_skirt_fr_02_r [{shapeType=Box, extentX=1.5, extentY=1.2251, extentZ=13, center=<-7 0 0>, rotation=<rot 90 0 0>}]
 dyn_skirt_fr_03 [{shapeType=Box, extentX=1.5, extentY=1.2251, extentZ=13, center=<7 0 0>, rotation=<rot 90 0 0>}]
 dyn_skirt_fr_03_l [{shapeType=Box, extentX=1.5, extentY=1.2251, extentZ=13, center=<7 0 0>, rotation=<rot 90 0 0>}]
 dyn_skirt_fr_03_r [{shapeType=Box, extentX=1.5, extentY=1.2251, extentZ=13, center=<-7 0 0>, rotation=<rot 90 0 0>}]
 dyn_skirt_frdiag_01_l [{shapeType=Box, extentX=1.5, extentY=8.5, extentZ=16.5851, center=<9.2404 0 -0.0001>, rotation=<rot 90 0 0>}]
 dyn_skirt_frdiag_01_r [{shapeType=Box, extentX=1.5, extentY=8.5, extentZ=16.5851, center=<-9.2404 0 0>, rotation=<rot 90 0 0>}]
 dyn_skirt_frdiag_02_l [{shapeType=Box, extentX=1.5, extentY=8.5, extentZ=13, center=<7 0 0>, rotation=<rot 90 0 0>}]
 dyn_skirt_frdiag_02_r [{shapeType=Box, extentX=3.1218, extentY=18.462, extentZ=15.6123, center=<-7.3001 0.0927 -2.4097>, rotation=<rot 82.8763 7.6649 6.9234>}]
 dyn_skirt_frdiag_03_l [{shapeType=Box, extentX=1.5, extentY=8.5, extentZ=11.4905, center=<6.7665 0 0>, rotation=<rot 90 0 0>}]
 dyn_skirt_frdiag_03_r [{shapeType=Box, extentX=2.5, extentY=19, extentZ=12.5692, center=<-6.7665 0 -1.2753>, rotation=<rot 90 0 0>}]
 dyn_skirt_frdiag_04_l [{shapeType=Box, extentX=1.5, extentY=8.5, extentZ=6.9211, center=<3.4203 0 -0.0001>, rotation=<rot 90 0 0>}]
 dyn_skirt_frdiag_04_r [{shapeType=Box, extentX=2.5, extentY=19, extentZ=7.9997, center=<-3.4203 0 -1.2753>, rotation=<rot 90 0 0>}]
 dyn_skirt_out_01_l [{shapeType=Box, extentX=1.5, extentY=8.5, extentZ=10.4159, center=<5.5995 0 0>, rotation=<rot 90 0 0>}]
 dyn_skirt_out_01_r [{shapeType=Box, extentX=1.5, extentY=8.5, extentZ=10.4159, center=<-5.5995 0 0>, rotation=<rot 90 0 0>}]
 dyn_skirt_out_01a_l [{shapeType=Box, extentX=2.0715, extentY=20.1965, extentZ=10.9874, center=<6.1609 0 0>, rotation=<rot 90 0 0>}]
 dyn_skirt_out_01a_r [{shapeType=Box, extentX=2.0715, extentY=20.1965, extentZ=10.9874, center=<-6.1609 0 0>, rotation=<rot 90 0 0>}]
 dyn_skirt_out_02_l [{shapeType=Box, extentX=1.5, extentY=19.625, extentZ=10.5792, center=<5.7881 0 0>, rotation=<rot 90 0 0>}]
 dyn_skirt_out_02_r [{shapeType=Box, extentX=2.5, extentY=19.625, extentZ=10.5792, center=<-5.7881 0 -1.9482>, rotation=<rot 90 0 0>}]
 dyn_skirt_out_02a_l [{shapeType=Box, extentX=1.5, extentY=19.625, extentZ=12.6201, center=<6.5988 0 0>, rotation=<rot 90 0 0>}]
 dyn_skirt_out_02a_r [{shapeType=Box, extentX=3.5, extentY=19.125, extentZ=12.6201, center=<-6.5988 0 -2.7162>, rotation=<rot 90 0 0>}]
 dyn_skirt_out_03_l [{shapeType=Box, extentX=1.5, extentY=19.625, extentZ=4.7937, center=<2.6035 0 0.0001>, rotation=<rot 90 0 0>}]
 dyn_skirt_out_03_r [{shapeType=Box, extentX=4.5, extentY=19.125, extentZ=4.7937, center=<-2.6035 0 -2.1518>, rotation=<rot 90 0 0>}]
 foot_l [{shapeType=Capsule, length=15.1506, radius=9.0978, center=<-3.5632 5.8205 -0.2581>, rotation=<rot -1.0393 6.4548 -84.0947>}]
 foot_r [{shapeType=Capsule, length=20.2474, radius=7.1941, center=<3.5632 -5.8205 0.2581>, rotation=<rot -1.0393 6.4548 -84.0947>}]
 pelvis [
  {shapeType=Capsule, length=7.75, radius=8.25, center=<0.0001 -2.6532 0>}
  {shapeType=Capsule, length=16.4237, radius=5.4668, center=<2.0146 -1.8959 0>}
  {shapeType=Capsule, length=11.4135, radius=8.5345, center=<-5.184 -3.5952 0>}
 ]
 thigh_l [
  {shapeType=Capsule, length=38.5156, radius=8.5978, center=<-26.894 -1.5716 -0.3412>, rotation=<rot 0 91.7591 -88.2002>}
  {shapeType=Capsule, length=15.3127, radius=6.7989, center=<-21.2939 1.8268 -1.3115>, rotation=<rot 89.5335 -180 -180>}
 ]
 thigh_r [
  {shapeType=Capsule, length=36.7187, radius=10.3947, center=<26.894 1.5716 0.3412>, rotation=<rot 0 91.7591 -88.2002>}
  {shapeType=Capsule, length=15.3127, radius=6.7989, center=<21.2939 -1.8268 1.3115>, rotation=<rot 89.5335 -180 -180>}
 ]
}
constraints [
 {bone1Name=thigh_l, bone2Name=pelvis, swing1LimitDegrees=45, swing2LimitDegrees=45, twistLimitDegrees=45}
 {bone1Name=calf_l, bone2Name=thigh_l, swing1LimitDegrees=45, swing2LimitDegrees=45, twistLimitDegrees=45}
 {bone1Name=thigh_r, bone2Name=pelvis, swing1LimitDegrees=45, swing2LimitDegrees=45, twistLimitDegrees=45}
 {bone1Name=calf_r, bone2Name=thigh_r, swing1LimitDegrees=45, swing2LimitDegrees=45, twistLimitDegrees=45}
 {bone1Name=dyn_skirt_out_02_l, bone2Name=dyn_skirt_out_01a_l, swing1LimitDegrees=40, swing2LimitDegrees=40, twistLimitDegrees=5}
 {bone1Name=dyn_skirt_fr_01_l, bone2Name=pelvis, swing1LimitDegrees=80, swing2LimitDegrees=80, twistLimitDegrees=5}
 {bone1Name=dyn_skirt_bk_02_l, bone2Name=dyn_skirt_bk_01a_l, swing1LimitDegrees=40, swing2LimitDegrees=40, twistLimitDegrees=5}
 {bone1Name=dyn_skirt_frdiag_02_l, bone2Name=dyn_skirt_frdiag_01_l, swing1LimitDegrees=5, swing2LimitDegrees=40, twistLimitDegrees=1}
 {bone1Name=dyn_skirt_fr_01, bone2Name=pelvis, swing1LimitDegrees=80, swing2LimitDegrees=80, twistLimitDegrees=5}
 {bone1Name=dyn_skirt_bk_02, bone2Name=dyn_skirt_bk_01a, swing1LimitDegrees=40, swing2LimitDegrees=40, twistLimitDegrees=5}
 {bone1Name=dyn_skirt_out_02_r, bone2Name=dyn_skirt_out_01a_r, swing1LimitDegrees=40, swing2LimitDegrees=40, twistLimitDegrees=5}
 {bone1Name=dyn_skirt_fr_01_r, bone2Name=pelvis, swing1LimitDegrees=80, swing2LimitDegrees=80, twistLimitDegrees=5}
 {bone1Name=dyn_skirt_bk_02_r, bone2Name=dyn_skirt_bk_01a_r, swing1LimitDegrees=40, swing2LimitDegrees=40, twistLimitDegrees=5}
 {bone1Name=dyn_skirt_frdiag_02_r, bone2Name=dyn_skirt_frdiag_01_r, swing1LimitDegrees=10, swing2LimitDegrees=40, twistLimitDegrees=5}
 {bone1Name=dyn_skirt_fr_02_l, bone2Name=dyn_skirt_fr_01_l, swing1LimitDegrees=80, swing2LimitDegrees=80, twistLimitDegrees=5}
 {bone1Name=dyn_skirt_fr_03_l, bone2Name=dyn_skirt_fr_02_l, swing1LimitDegrees=80, swing2LimitDegrees=80, twistLimitDegrees=5}
 {bone1Name=dyn_skirt_out_01_l, bone2Name=pelvis, swing1LimitDegrees=40, swing2LimitDegrees=40, twistLimitDegrees=5}
 {bone1Name=dyn_skirt_out_03_l, bone2Name=dyn_skirt_out_02a_l, swing1LimitDegrees=40, swing2LimitDegrees=40, twistLimitDegrees=5}
 {bone1Name=dyn_skirt_bk_01_l, bone2Name=pelvis, swing1LimitDegrees=40, swing2LimitDegrees=40, twistLimitDegrees=5}
 {bone1Name=dyn_skirt_bk_03_l, bone2Name=dyn_skirt_bk_02a_l, swing1LimitDegrees=40, swing2LimitDegrees=40, twistLimitDegrees=5}
 {bone1Name=dyn_skirt_frdiag_01_l, bone2Name=pelvis, swing1LimitDegrees=20, swing2LimitDegrees=40, twistLimitDegrees=1}
 {bone1Name=dyn_skirt_frdiag_03_l, bone2Name=dyn_skirt_frdiag_02_l, swing1LimitDegrees=30, swing2LimitDegrees=40, twistLimitDegrees=5}
 {bone1Name=dyn_skirt_frdiag_04_l, bone2Name=dyn_skirt_frdiag_03_l, swing1LimitDegrees=30, swing2LimitDegrees=40, twistLimitDegrees=5}
 {bone1Name=dyn_skirt_fr_02, bone2Name=dyn_skirt_fr_01, swing1LimitDegrees=80, swing2LimitDegrees=80, twistLimitDegrees=5}
 {bone1Name=dyn_skirt_fr_03, bone2Name=dyn_skirt_fr_02, swing1LimitDegrees=80, swing2LimitDegrees=80, twistLimitDegrees=5}
 {bone1Name=dyn_skirt_bk_01, bone2Name=pelvis, swing1LimitDegrees=40, swing2LimitDegrees=40, twistLimitDegrees=5}
 {bone1Name=dyn_skirt_bk_03, bone2Name=dyn_skirt_bk_02a, swing1LimitDegrees=40, swing2LimitDegrees=40, twistLimitDegrees=5}
 {bone1Name=dyn_skirt_out_01_r, bone2Name=pelvis, swing1LimitDegrees=40, swing2LimitDegrees=40, twistLimitDegrees=5}
 {bone1Name=dyn_skirt_out_03_r, bone2Name=dyn_skirt_out_02a_r, swing1LimitDegrees=40, swing2LimitDegrees=40, twistLimitDegrees=5}
 {bone1Name=dyn_skirt_fr_02_r, bone2Name=dyn_skirt_fr_01_r, swing1LimitDegrees=80, swing2LimitDegrees=80, twistLimitDegrees=5}
 {bone1Name=dyn_skirt_fr_03_r, bone2Name=dyn_skirt_fr_02_r, swing1LimitDegrees=80, swing2LimitDegrees=80, twistLimitDegrees=5}
 {bone1Name=dyn_skirt_bk_01_r, bone2Name=pelvis, swing1LimitDegrees=40, swing2LimitDegrees=40, twistLimitDegrees=5}
 {bone1Name=dyn_skirt_bk_03_r, bone2Name=dyn_skirt_bk_02a_r, swing1LimitDegrees=40, swing2LimitDegrees=40, twistLimitDegrees=5}
 {bone1Name=dyn_skirt_frdiag_01_r, bone2Name=pelvis, swing1LimitDegrees=20, swing2LimitDegrees=40, twistLimitDegrees=5}
 {bone1Name=dyn_skirt_frdiag_03_r, bone2Name=dyn_skirt_frdiag_02_r, swing1LimitDegrees=10, swing2LimitDegrees=40, twistLimitDegrees=5}
 {bone1Name=dyn_skirt_frdiag_04_r, bone2Name=dyn_skirt_frdiag_03_r, swing1LimitDegrees=10, swing2LimitDegrees=40, twistLimitDegrees=5}
 {
  bone1Name=foot_l
  bone2Name=calf_l
  swing1LimitDegrees=45
  swing1Motion=Free
  swing2LimitDegrees=45
  swing2Motion=Free
  twistLimitDegrees=45
  twistMotion=Free
 }
 {
  bone1Name=foot_r
  bone2Name=calf_r
  swing1LimitDegrees=45
  swing1Motion=Free
  swing2LimitDegrees=45
  swing2Motion=Free
  twistLimitDegrees=45
  twistMotion=Free
 }
 {bone1Name=dyn_skirt_out_01a_l, bone2Name=dyn_skirt_out_01_l, swing1LimitDegrees=40, swing2LimitDegrees=40, twistLimitDegrees=5}
 {bone1Name=dyn_skirt_out_02a_l, bone2Name=dyn_skirt_out_02_l, swing1LimitDegrees=40, swing2LimitDegrees=40, twistLimitDegrees=5}
 {bone1Name=dyn_skirt_bk_01a_l, bone2Name=dyn_skirt_bk_01_l, swing1LimitDegrees=40, swing2LimitDegrees=40, twistLimitDegrees=5}
 {bone1Name=dyn_skirt_bk_02a_l, bone2Name=dyn_skirt_bk_02_l, swing1LimitDegrees=40, swing2LimitDegrees=40, twistLimitDegrees=5}
 {bone1Name=dyn_skirt_out_01a_r, bone2Name=dyn_skirt_out_01_r, swing1LimitDegrees=40, swing2LimitDegrees=40, twistLimitDegrees=5}
 {bone1Name=dyn_skirt_out_02a_r, bone2Name=dyn_skirt_out_02_r, swing1LimitDegrees=40, swing2LimitDegrees=40, twistLimitDegrees=5}
 {bone1Name=dyn_skirt_bk_01a_r, bone2Name=dyn_skirt_bk_01_r, swing1LimitDegrees=40, swing2LimitDegrees=40, twistLimitDegrees=5}
 {bone1Name=dyn_skirt_bk_02a_r, bone2Name=dyn_skirt_bk_02_r, swing1LimitDegrees=40, swing2LimitDegrees=40, twistLimitDegrees=5}
 {bone1Name=dyn_skirt_bk_01a, bone2Name=dyn_skirt_bk_01, swing1LimitDegrees=40, swing2LimitDegrees=40, twistLimitDegrees=5}
 {bone1Name=dyn_skirt_bk_02a, bone2Name=dyn_skirt_bk_02, swing1LimitDegrees=40, swing2LimitDegrees=40, twistLimitDegrees=5}
 {bone1Name=dyn_skirt_bkmid_01_l, bone2Name=pelvis, swing1LimitDegrees=40, swing2LimitDegrees=40, twistLimitDegrees=5}
 {bone1Name=dyn_skirt_bkmid_01a_l, bone2Name=dyn_skirt_bkmid_01_l, swing1LimitDegrees=40, swing2LimitDegrees=40, twistLimitDegrees=5}
 {bone1Name=dyn_skirt_bkmid_02_l, bone2Name=dyn_skirt_bkmid_01a_l, swing1LimitDegrees=40, swing2LimitDegrees=40, twistLimitDegrees=5}
 {bone1Name=dyn_skirt_bkmid_02a_l, bone2Name=dyn_skirt_bkmid_02_l, swing1LimitDegrees=40, swing2LimitDegrees=40, twistLimitDegrees=5}
 {bone1Name=dyn_skirt_bkmid_03_l, bone2Name=dyn_skirt_bkmid_02a_l, swing1LimitDegrees=40, swing2LimitDegrees=40, twistLimitDegrees=5}
 {bone1Name=dyn_skirt_bkdiag_01_l, bone2Name=pelvis, swing1LimitDegrees=40, swing2LimitDegrees=40, twistLimitDegrees=5}
 {bone1Name=dyn_skirt_bkdiag_01a_l, bone2Name=dyn_skirt_bkdiag_01_l, swing1LimitDegrees=40, swing2LimitDegrees=40, twistLimitDegrees=5}
 {bone1Name=dyn_skirt_bkdiag_02_l, bone2Name=dyn_skirt_bkdiag_01a_l, swing1LimitDegrees=40, swing2LimitDegrees=40, twistLimitDegrees=5}
 {bone1Name=dyn_skirt_bkdiag_02a_l, bone2Name=dyn_skirt_bkdiag_02_l, swing1LimitDegrees=40, swing2LimitDegrees=40, twistLimitDegrees=5}
 {bone1Name=dyn_skirt_bkdiag_03_l, bone2Name=dyn_skirt_bkdiag_02a_l, swing1LimitDegrees=40, swing2LimitDegrees=40, twistLimitDegrees=5}
 {bone1Name=dyn_skirt_bkdiag_01a_r, bone2Name=dyn_skirt_bkdiag_01_r, swing1LimitDegrees=40, swing2LimitDegrees=40, twistLimitDegrees=5}
 {bone1Name=dyn_skirt_bkdiag_01_r, bone2Name=pelvis, swing1LimitDegrees=40, swing2LimitDegrees=40, twistLimitDegrees=5}
 {bone1Name=dyn_skirt_bkdiag_02_r, bone2Name=dyn_skirt_bkdiag_01a_r, swing1LimitDegrees=40, swing2LimitDegrees=40, twistLimitDegrees=5}
 {bone1Name=dyn_skirt_bkdiag_02a_r, bone2Name=dyn_skirt_bkdiag_02_r, swing1LimitDegrees=40, swing2LimitDegrees=40, twistLimitDegrees=5}
 {bone1Name=dyn_skirt_bkdiag_03_r, bone2Name=dyn_skirt_bkdiag_02a_r, swing1LimitDegrees=40, swing2LimitDegrees=40, twistLimitDegrees=5}
 {bone1Name=dyn_skirt_bkmid_01_r, bone2Name=pelvis, swing1LimitDegrees=40, swing2LimitDegrees=40, twistLimitDegrees=5}
 {bone1Name=dyn_skirt_bkmid_01a_r, bone2Name=dyn_skirt_bkmid_01_r, swing1LimitDegrees=40, swing2LimitDegrees=40, twistLimitDegrees=5}
 {bone1Name=dyn_skirt_bkmid_02_r, bone2Name=dyn_skirt_bkmid_01a_r, swing1LimitDegrees=40, swing2LimitDegrees=40, twistLimitDegrees=5}
 {bone1Name=dyn_skirt_bkmid_02a_r, bone2Name=dyn_skirt_bkmid_02_r, swing1LimitDegrees=40, swing2LimitDegrees=40, twistLimitDegrees=5}
 {bone1Name=dyn_skirt_bkmid_03_r, bone2Name=dyn_skirt_bkmid_02a_r, swing1LimitDegrees=40, swing2LimitDegrees=40, twistLimitDegrees=5}
 {
  bone1Name=dyn_skirt_bkmid_01_r
  bone2Name=dyn_skirt_bk_01
  swing1LimitDegrees=45
  swing1Motion=Free
  swing2LimitDegrees=45
  swing2Motion=Free
  twistLimitDegrees=45
  twistMotion=Free
 }
 {
  bone1Name=dyn_skirt_bkmid_01a_r
  bone2Name=dyn_skirt_bk_01a
  swing1LimitDegrees=45
  swing1Motion=Free
  swing2LimitDegrees=45
  swing2Motion=Free
  twistLimitDegrees=45
  twistMotion=Free
 }
 {
  bone1Name=dyn_skirt_bkmid_02_r
  bone2Name=dyn_skirt_bk_02
  swing1LimitDegrees=45
  swing1Motion=Free
  swing2LimitDegrees=45
  swing2Motion=Free
  twistLimitDegrees=45
  twistMotion=Free
 }
 {
  bone1Name=dyn_skirt_bkmid_02a_r
  bone2Name=dyn_skirt_bk_02a
  swing1LimitDegrees=45
  swing1Motion=Free
  swing2LimitDegrees=45
  swing2Motion=Free
  twistLimitDegrees=45
  twistMotion=Free
 }
 {
  bone1Name=dyn_skirt_bkmid_03_r
  bone2Name=dyn_skirt_bk_03
  swing1LimitDegrees=45
  swing1Motion=Free
  swing2LimitDegrees=45
  swing2Motion=Free
  twistLimitDegrees=45
  twistMotion=Free
 }
 {
  bone1Name=dyn_skirt_bk_01_r
  bone2Name=dyn_skirt_bkmid_01_r
  swing1LimitDegrees=45
  swing1Motion=Free
  swing2LimitDegrees=45
  swing2Motion=Free
  twistLimitDegrees=45
  twistMotion=Free
 }
 {
  bone1Name=dyn_skirt_bk_01a_r
  bone2Name=dyn_skirt_bkmid_01a_r
  swing1LimitDegrees=45
  swing1Motion=Free
  swing2LimitDegrees=45
  swing2Motion=Free
  twistLimitDegrees=45
  twistMotion=Free
 }
 {
  bone1Name=dyn_skirt_bk_02_r
  bone2Name=dyn_skirt_bkmid_02_r
  swing1LimitDegrees=45
  swing1Motion=Free
  swing2LimitDegrees=45
  swing2Motion=Free
  twistLimitDegrees=45
  twistMotion=Free
 }
 {
  bone1Name=dyn_skirt_bk_02a_r
  bone2Name=dyn_skirt_bkmid_02a_r
  swing1LimitDegrees=45
  swing1Motion=Free
  swing2LimitDegrees=45
  swing2Motion=Free
  twistLimitDegrees=45
  twistMotion=Free
 }
 {
  bone1Name=dyn_skirt_bk_03_r
  bone2Name=dyn_skirt_bkmid_03_r
  swing1LimitDegrees=45
  swing1Motion=Free
  swing2LimitDegrees=45
  swing2Motion=Free
  twistLimitDegrees=45
  twistMotion=Free
 }
 {
  bone1Name=dyn_skirt_bkmid_01a_l
  bone2Name=dyn_skirt_bk_01a
  swing1LimitDegrees=45
  swing1Motion=Free
  swing2LimitDegrees=45
  swing2Motion=Free
  twistLimitDegrees=45
  twistMotion=Free
 }
 {
  bone1Name=dyn_skirt_bkmid_02_l
  bone2Name=dyn_skirt_bk_02
  swing1LimitDegrees=45
  swing1Motion=Free
  swing2LimitDegrees=45
  swing2Motion=Free
  twistLimitDegrees=45
  twistMotion=Free
 }
 {
  bone1Name=dyn_skirt_bkmid_02a_l
  bone2Name=dyn_skirt_bk_02a
  swing1LimitDegrees=45
  swing1Motion=Free
  swing2LimitDegrees=45
  swing2Motion=Free
  twistLimitDegrees=45
  twistMotion=Free
 }
 {
  bone1Name=dyn_skirt_bkmid_03_l
  bone2Name=dyn_skirt_bk_03
  swing1LimitDegrees=45
  swing1Motion=Free
  swing2LimitDegrees=45
  swing2Motion=Free
  twistLimitDegrees=45
  twistMotion=Free
 }
 {
  bone1Name=dyn_skirt_bkmid_01_l
  bone2Name=dyn_skirt_bk_01
  swing1LimitDegrees=45
  swing1Motion=Free
  swing2LimitDegrees=45
  swing2Motion=Free
  twistLimitDegrees=45
  twistMotion=Free
 }
 {
  bone1Name=dyn_skirt_bk_01_l
  bone2Name=dyn_skirt_bkmid_01_l
  swing1LimitDegrees=45
  swing1Motion=Free
  swing2LimitDegrees=45
  swing2Motion=Free
  twistLimitDegrees=45
  twistMotion=Free
 }
 {
  bone1Name=dyn_skirt_bk_01a_l
  bone2Name=dyn_skirt_bkmid_01a_l
  swing1LimitDegrees=45
  swing1Motion=Free
  swing2LimitDegrees=45
  swing2Motion=Free
  twistLimitDegrees=45
  twistMotion=Free
 }
 {
  bone1Name=dyn_skirt_bk_02_l
  bone2Name=dyn_skirt_bkmid_02_l
  swing1LimitDegrees=45
  swing1Motion=Free
  swing2LimitDegrees=45
  swing2Motion=Free
  twistLimitDegrees=45
  twistMotion=Free
 }
 {
  bone1Name=dyn_skirt_bk_02a_l
  bone2Name=dyn_skirt_bkmid_02a_l
  swing1LimitDegrees=45
  swing1Motion=Free
  swing2LimitDegrees=45
  swing2Motion=Free
  twistLimitDegrees=45
  twistMotion=Free
 }
 {
  bone1Name=dyn_skirt_bk_03_l
  bone2Name=dyn_skirt_bkmid_03_l
  swing1LimitDegrees=45
  swing1Motion=Free
  swing2LimitDegrees=45
  swing2Motion=Free
  twistLimitDegrees=45
  twistMotion=Free
 }
 {
  bone1Name=dyn_skirt_fr_03_l
  bone2Name=dyn_skirt_frdiag_03_l
  swing1LimitDegrees=45
  swing1Motion=Free
  swing2LimitDegrees=45
  swing2Motion=Free
  twistLimitDegrees=45
  twistMotion=Free
 }
 {
  bone1Name=dyn_skirt_fr_03_r
  bone2Name=dyn_skirt_frdiag_03_r
  swing1LimitDegrees=45
  swing1Motion=Free
  swing2LimitDegrees=45
  swing2Motion=Free
  twistLimitDegrees=45
  twistMotion=Free
 }
 {
  bone1Name=dyn_skirt_fr_02_r
  bone2Name=dyn_skirt_frdiag_02_r
  swing1LimitDegrees=45
  swing1Motion=Free
  swing2LimitDegrees=45
  swing2Motion=Free
  twistLimitDegrees=45
  twistMotion=Free
 }
 {
  bone1Name=dyn_skirt_fr_02_l
  bone2Name=dyn_skirt_frdiag_02_l
  swing1LimitDegrees=45
  swing1Motion=Free
  swing2LimitDegrees=45
  swing2Motion=Free
  twistLimitDegrees=45
  twistMotion=Free
 }
 {
  bone1Name=dyn_skirt_frdiag_03_l
  bone2Name=dyn_skirt_out_02a_l
  swing1LimitDegrees=45
  swing1Motion=Free
  swing2LimitDegrees=45
  swing2Motion=Free
  twistLimitDegrees=45
  twistMotion=Free
 }
 {
  bone1Name=dyn_skirt_frdiag_03_r
  bone2Name=dyn_skirt_out_02a_r
  swing1LimitDegrees=45
  swing1Motion=Free
  swing2LimitDegrees=45
  swing2Motion=Free
  twistLimitDegrees=45
  twistMotion=Free
 }
 {
  bone1Name=dyn_skirt_bkdiag_02a_r
  bone2Name=dyn_skirt_out_02a_r
  swing1LimitDegrees=45
  swing1Motion=Free
  swing2LimitDegrees=45
  swing2Motion=Free
  twistLimitDegrees=45
  twistMotion=Free
 }
 {
  bone1Name=dyn_skirt_bkdiag_02a_r
  bone2Name=dyn_skirt_bk_02a_r
  swing1LimitDegrees=45
  swing1Motion=Free
  swing2LimitDegrees=45
  swing2Motion=Free
  twistLimitDegrees=45
  twistMotion=Free
 }
]
