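# Physics asset for the hanging necklace and hourglass pendant on the F_MED_Plague female character body.
# Capsule bodies form a chain from neck_01 through two necklace bones to the hourglass, joined by constraints with moderate swing and tight twist limits.
# Larger capsules on spine_04 and spine_05 approximate the upper torso, with free-rotating constraints linking those spine bodies and the neck.
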PhysicsAsset /BRCosmetics/Characters/Player/Female/Medium/Bodies/F_MED_Plague/Meshes/F_MED_Plague_Necklace_Dynamics
bodies {
 dyn_hourglass [{shapeType=Capsule, length=3, radius=1.75, center=<-0.0001 0 -3.7>}]
 dyn_necklace_01 [{shapeType=Capsule, length=8.5938, radius=1.25, center=<0 0 -10.3581>}]
 dyn_necklace_02 [{shapeType=Capsule, length=8.5938, radius=1.25, center=<0 0 -4.1624>}]
 neck_01 [{shapeType=Capsule, length=0.505, radius=0.505, rotation=<rot 0 -90 0>}]
 spine_04 [{shapeType=Capsule, length=14, radius=5, center=<1.3134 8.9541 0>}]
 spine_05 [
  {shapeType=Capsule, length=27.25, radius=2.75, center=<4.847 4.8975 0>}
  {shapeType=Capsule, length=15, radius=15, center=<-6.9451 -2.0096 0>}
 ]
}
constraints [
 {bone1Name=dyn_hourglass, bone2Name=dyn_necklace_02, swing1LimitDegrees=35, swing2LimitDegrees=35, twistLimitDegrees=5}
 {bone1Name=dyn_necklace_02, bone2Name=dyn_necklace_01, swing1LimitDegrees=35, swing2LimitDegrees=35, twistLimitDegrees=5}
 {bone1Name=dyn_necklace_01, bone2Name=neck_01, swing1LimitDegrees=35, swing2LimitDegrees=35, twistLimitDegrees=5}
 {
  bone1Name=neck_01
  bone2Name=spine_05
  swing1LimitDegrees=45
  swing1Motion=Free
  swing2LimitDegrees=45
  swing2Motion=Free
  twistLimitDegrees=45
  twistMotion=Free
 }
 {
  bone1Name=spine_05
  bone2Name=spine_04
  swing1LimitDegrees=45
  swing1Motion=Free
  swing2LimitDegrees=45
  swing2Motion=Free
  twistLimitDegrees=45
  twistMotion=Free
 }
]
