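# Post-process style AnimBlueprint for the F_MED_Plague body cosmetic, adding secondary motion to skirt, necklace and forearm tassels.
# It copies the parent mesh pose, then constrains front skirt bones to thighs, with corrective offsets weighted by aim and crouch variables.
# Chained AnimDynamics and RigidBody nodes simulate dangling parts, sharing a skydiving force; dyn_root takes a world-space rotation.
# Center front skirt bones are constrained between left and right chains before simulating.
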AnimBlueprint /BRCosmetics/Characters/Player/Female/Medium/Bodies/F_MED_Plague/Meshes/F_MED_Plague_AnimBP
target_skeleton=&/F_MED_Plague_Skeleton
graphs [
 AnimGraph {nodes [
   n1 = Anim Dynamics - Bone: dyn_skirt_frdiag_01_l(Alpha=1, ExternalForce="0,0,0") {
    angularDampingOverride=0.98
    bGravityOverrideInSimSpace=true
    bOverrideAngularDamping=true
    bUseGravityOverride=true
    numSolverIterationsPostUpdate=2
    numSolverIterationsPreUpdate=8
    simulationSpace=RootRelative
    alphaBoolBlend {blendOption=Linear}
    boundBone {boneName=dyn_skirt_frdiag_01_l}
    gravityOverride {z=-980}
    physicsBodyDefinitions [{
      sphereCollisionRadius=0
      boundBone {boneName=dyn_skirt_frdiag_01_l}
      boxExtents=<10 20 20>
      constraintSetup {twistAxis=AxisZ, angularLimitsMax=<0 0 2>, angularLimitsMin=<0 -30 -2>, angularTarget=<0 0 0>}
      localJointOffset=<3 0 0>
     }]
   }
   n2 = Anim Dynamics - Bone: dyn_skirt_frdiag_02_l(ComponentPose=@n1, Alpha=1, ExternalForce="0,0,0") {
    angularDampingOverride=0.98
    bGravityOverrideInSimSpace=true
    bOverrideAngularDamping=true
    bUseGravityOverride=true
    simulationSpace=RootRelative
    alphaBoolBlend {blendOption=Linear}
    boundBone {boneName=dyn_skirt_frdiag_02_l}
    gravityOverride {z=-980}
    physicsBodyDefinitions [{
      sphereCollisionRadius=0
      boundBone {boneName=dyn_skirt_frdiag_02_l}
      boxExtents=<10 20 20>
      constraintSetup {twistAxis=AxisZ, angularLimitsMax=<10 5 2>, angularLimitsMin=<-10 -30 -2>, angularTarget=<0 0 0>}
      localJointOffset=<3 0 0>
     }]
   }
   n3 = Anim Dynamics - Bone: dyn_skirt_frdiag_03_l(ComponentPose=@n2, Alpha=1, ExternalForce="0,0,0") {
    angularDampingOverride=0.98
    bGravityOverrideInSimSpace=true
    bOverrideAngularDamping=true
    bUseGravityOverride=true
    simulationSpace=RootRelative
    alphaBoolBlend {blendOption=Linear}
    boundBone {boneName=dyn_skirt_frdiag_03_l}
    gravityOverride {z=-980}
    physicsBodyDefinitions [{
      sphereCollisionRadius=0
      boundBone {boneName=dyn_skirt_frdiag_03_l}
      boxExtents=<10 20 20>
      constraintSetup {twistAxis=AxisZ, angularLimitsMax=<10 15 2>, angularLimitsMin=<-10 -30 -2>, angularTarget=<0 0 0>}
      localJointOffset=<3 0 0>
     }]
   }
   n4 = Anim Dynamics - Bone: dyn_skirt_frdiag_04_l(ComponentPose=@n3, Alpha=1, ExternalForce="0,0,0") {
    angularDampingOverride=0.98
    bGravityOverrideInSimSpace=true
    bOverrideAngularDamping=true
    bUseGravityOverride=true
    simulationSpace=RootRelative
    alphaBoolBlend {blendOption=Linear}
    boundBone {boneName=dyn_skirt_frdiag_04_l}
    gravityOverride {z=-980}
    physicsBodyDefinitions [{
      sphereCollisionRadius=0
      boundBone {boneName=dyn_skirt_frdiag_04_l}
      boxExtents=<10 20 20>
      constraintSetup {twistAxis=AxisZ, angularLimitsMax=<10 15 2>, angularLimitsMin=<-10 -30 -2>, angularTarget=<0 0 0>}
      localJointOffset=<3 0 0>
     }]
   }
   n5 = Anim Dynamics - Bone: dyn_skirt_frdiag_01_r(ComponentPose=@n4, Alpha=1, ExternalForce="0,0,0") {
    angularDampingOverride=0.98
    bGravityOverrideInSimSpace=true
    bOverrideAngularDamping=true
    bUseGravityOverride=true
    simulationSpace=RootRelative
    alphaBoolBlend {blendOption=Linear}
    boundBone {boneName=dyn_skirt_frdiag_01_r}
    gravityOverride {z=-980}
    physicsBodyDefinitions [{
      sphereCollisionRadius=0
      boundBone {boneName=dyn_skirt_frdiag_01_r}
      boxExtents=<10 20 20>
      constraintSetup {twistAxis=AxisZ, angularLimitsMax=<0 0 2>, angularLimitsMin=<0 -30 -2>, angularTarget=<0 0 0>}
      localJointOffset=<-3 0 0>
     }]
   }
   n6 = Anim Dynamics - Bone: dyn_skirt_frdiag_02_r(ComponentPose=@n5, Alpha=1, ExternalForce="0,0,0") {
    angularDampingOverride=0.98
    bGravityOverrideInSimSpace=true
    bOverrideAngularDamping=true
    bUseGravityOverride=true
    simulationSpace=RootRelative
    alphaBoolBlend {blendOption=Linear}
    boundBone {boneName=dyn_skirt_frdiag_02_r}
    gravityOverride {z=-980}
    physicsBodyDefinitions [{
      sphereCollisionRadius=0
      boundBone {boneName=dyn_skirt_frdiag_02_r}
      boxExtents=<10 20 20>
      constraintSetup {twistAxis=AxisZ, angularLimitsMax=<10 5 2>, angularLimitsMin=<-10 -30 -2>, angularTarget=<0 0 0>}
      localJointOffset=<-3 0 0>
     }]
   }
   n7 = Anim Dynamics - Bone: dyn_skirt_frdiag_03_r(ComponentPose=@n6, Alpha=1, ExternalForce="0,0,0") {
    angularDampingOverride=0.98
    bGravityOverrideInSimSpace=true
    bOverrideAngularDamping=true
    bUseGravityOverride=true
    simulationSpace=RootRelative
    alphaBoolBlend {blendOption=Linear}
    boundBone {boneName=dyn_skirt_frdiag_03_r}
    gravityOverride {z=-980}
    physicsBodyDefinitions [{
      sphereCollisionRadius=0
      boundBone {boneName=dyn_skirt_frdiag_03_r}
      boxExtents=<10 20 20>
      constraintSetup {twistAxis=AxisZ, angularLimitsMax=<10 15 2>, angularLimitsMin=<-10 -30 -2>, angularTarget=<0 0 0>}
      localJointOffset=<-3 0 0>
     }]
   }
   Anim Dynamics - Bone: dyn_skirt_frdiag_04_r(ComponentPose=@n7, Alpha=1, ExternalForce="0,0,0") {
    angularDampingOverride=0.98
    bGravityOverrideInSimSpace=true
    bOverrideAngularDamping=true
    bUseGravityOverride=true
    bUseSphericalLimits=true
    simulationSpace=RootRelative
    alphaBoolBlend {blendOption=Linear}
    boundBone {boneName=dyn_skirt_frdiag_04_r}
    gravityOverride {z=-980}
    physicsBodyDefinitions [{
      collisionType=CustomSphere
      sphereCollisionRadius=1
      boundBone {boneName=dyn_skirt_frdiag_04_r}
      boxExtents=<10 20 20>
      constraintSetup {twistAxis=AxisZ, angularLimitsMax=<10 20 2>, angularLimitsMin=<-10 -30 -2>, angularTarget=<0 0 0>}
      localJointOffset=<-3 0 0>
     }]
    sphericalLimits [{limitRadius=7, drivingBone {boneName=thigh_r}, sphereLocalOffset=<38 0 0>}]
   }
   n8 = Anim Dynamics - Bone: dyn_skirt_bk_01(Alpha=1, ExternalForce="0,0,0") {
    angularDampingOverride=0.98
    bGravityOverrideInSimSpace=true
    bOverrideAngularDamping=true
    bUseGravityOverride=true
    numSolverIterationsPostUpdate=2
    numSolverIterationsPreUpdate=8
    simulationSpace=RootRelative
    alphaBoolBlend {blendOption=Linear}
    boundBone {boneName=dyn_skirt_bk_01}
    gravityOverride {z=-980}
    physicsBodyDefinitions [{
      sphereCollisionRadius=0
      boundBone {boneName=dyn_skirt_bk_01}
      boxExtents=<10 20 20>
      constraintSetup {twistAxis=AxisZ, angularLimitsMax=<0 0 2>, angularLimitsMin=<0 -30 -2>, angularTarget=<0 0 0>}
      localJointOffset=<3 0 0>
     }]
   }
   n9 = Anim Dynamics - Bone: dyn_skirt_bk_02(ComponentPose=@n8, Alpha=1, ExternalForce="0,0,0") {
    angularDampingOverride=0.98
    bGravityOverrideInSimSpace=true
    bOverrideAngularDamping=true
    bUseGravityOverride=true
    simulationSpace=RootRelative
    alphaBoolBlend {blendOption=Linear}
    boundBone {boneName=dyn_skirt_bk_02}
    gravityOverride {z=-980}
    physicsBodyDefinitions [{
      sphereCollisionRadius=0
      boundBone {boneName=dyn_skirt_bk_02}
      boxExtents=<10 20 20>
      constraintSetup {twistAxis=AxisZ, angularLimitsMax=<10 15 2>, angularLimitsMin=<-10 -30 -2>, angularTarget=<0 0 0>}
      localJointOffset=<3 0 0>
     }]
   }
   Anim Dynamics - Bone: dyn_skirt_bk_03(ComponentPose=@n9, Alpha=1, ExternalForce="0,0,0") {
    angularDampingOverride=0.98
    bGravityOverrideInSimSpace=true
    bOverrideAngularDamping=true
    bUseGravityOverride=true
    simulationSpace=RootRelative
    alphaBoolBlend {blendOption=Linear}
    boundBone {boneName=dyn_skirt_bk_03}
    gravityOverride {z=-980}
    physicsBodyDefinitions [{
      sphereCollisionRadius=0
      boundBone {boneName=dyn_skirt_bk_03}
      boxExtents=<10 20 20>
      constraintSetup {twistAxis=AxisZ, angularLimitsMax=<10 20 2>, angularLimitsMin=<-10 -30 -2>, angularTarget=<0 0 0>}
      localJointOffset=<3 0 0>
     }]
   }
   "dyn_skirt_fr_01_r.translateX = thigh_r.rotateZ * -0.05 - Bone Driven Controller"!(ComponentPose:Component Space Pose Link Structure, Alpha:float=1) -> (Pose:Component Space Pose Link Structure) {
    is=title AnimGraphNode_BoneDrivenController
    bAffectTargetTranslationX=true
    multiplier=-0.05
    sourceComponent=RotationZ
    alphaBoolBlend {blendOption=Linear}
    sourceBone {boneName=thigh_r}
    targetBone {boneName=dyn_skirt_fr_01_r}
   }
   "dyn_skirt_fr_01_r.rotateY = curve(thigh_r.rotateZ) - Bone Driven Controller"!(ComponentPose:Component Space Pose Link Structure, Alpha:float=1) -> (Pose:Component Space Pose Link Structure) {
    is=title AnimGraphNode_BoneDrivenController
    bAffectTargetRotationY=true
    drivingCurve=&/r_thigh_curve
    multiplier=-0.5
    sourceComponent=RotationZ
    alphaBoolBlend {blendOption=Linear}
    sourceBone {boneName=thigh_r}
    targetBone {boneName=dyn_skirt_fr_01_r}
   }
   n10 = "dyn_skirt_fr_01_l.translateX = thigh_l.rotateZ * -0.01 - Bone Driven Controller"!(ComponentPose:Component Space Pose Link Structure, Alpha:float=1) -> (Pose:Component Space Pose Link Structure) {
    is=title AnimGraphNode_BoneDrivenController
    bAffectTargetTranslationX=true
    multiplier=-0.01
    sourceComponent=RotationZ
    alphaBoolBlend {blendOption=Linear}
    sourceBone {boneName=thigh_l}
    targetBone {boneName=dyn_skirt_fr_01_l}
   }
   "dyn_skirt_fr_01_l.translateZ = thigh_l.rotateZ * 0.09 - Bone Driven Controller"!(ComponentPose:Component Space Pose Link Structure=@n10, Alpha:float=1) -> (Pose:Component Space Pose Link Structure) {
    is=title AnimGraphNode_BoneDrivenController
    bAffectTargetTranslationZ=true
    multiplier=0.09
    sourceComponent=RotationZ
    alphaBoolBlend {blendOption=Linear}
    sourceBone {boneName=thigh_l}
    targetBone {boneName=dyn_skirt_fr_01_l}
   }
   "Transform (Modify) Bone - Bone: dyn_skirt_frdiag_01_l"(Alpha=1, Translation="2.120,0.472,3.739", Rotation="-12.662,-21.035,5.076", Scale="1,1,1") {
    rotationMode=BMM_Additive
    rotationSpace=BCS_BoneSpace
    translationMode=BMM_Additive
    translationSpace=BCS_BoneSpace
    alphaBoolBlend {blendOption=Linear}
    boneToModify {boneName=dyn_skirt_frdiag_01_l}
    rotation=<rot -12.662 -21.0348 5.0761>
    translation=<2.12 0.472 3.739>
   }
   "Transform (Modify) Bone - Bone: dyn_skirt_out_01_l"(Alpha=1, Translation="0,0,0", Rotation="0.948,36.323,39.295", Scale="1,1,1") {
    rotationMode=BMM_Additive
    rotationSpace=BCS_BoneSpace
    alphaBoolBlend {blendOption=Linear}
    boneToModify {boneName=dyn_skirt_out_01_l}
    rotation=<rot 0.948 36.323 39.2949>
   }
   n11 = "Transform (Modify) Bone - Bone: dyn_skirt_bk_01_l"(Alpha=1, Translation="0,0,0", Rotation="-19.568,17.170,-11.008", Scale="1,1,1") {
    rotationMode=BMM_Additive
    rotationSpace=BCS_BoneSpace
    alphaBoolBlend {blendOption=Linear}
    boneToModify {boneName=dyn_skirt_bk_01_l}
    rotation=<rot -19.568 17.17 -11.008>
   }
   n12 = "Transform (Modify) Bone - Bone: dyn_skirt_bk_01"(ComponentPose=@n11, Alpha=1, Translation="0,0,0", Rotation="-32,0,0", Scale="1,1,1") {
    rotationMode=BMM_Additive
    rotationSpace=BCS_BoneSpace
    alphaBoolBlend {blendOption=Linear}
    boneToModify {boneName=dyn_skirt_bk_01}
    rotation {pitch=-32}
   }
   "Transform (Modify) Bone - Bone: dyn_skirt_bk_01_r"(ComponentPose=@n12, Alpha=1, Translation="0,0,0", Rotation="-36.346,16.501,-21.121", Scale="1,1,1") {
    rotationMode=BMM_Additive
    rotationSpace=BCS_BoneSpace
    alphaBoolBlend {blendOption=Linear}
    boneToModify {boneName=dyn_skirt_bk_01_r}
    rotation=<rot -36.346 16.5012 -21.1209>
   }
   RigidBody() {
    bOverrideWorldGravity=true
    overridePhysicsAsset=&/F_MED_Plague_ArmDanglies_Dynamics
    simulationSpace=BaseBoneSpace
    alphaBoolBlend {blendOption=Linear}
    baseBoneRef {boneName=root}
    componentAppliedLinearAccClamp=<1000 1000 1000>
    componentLinearAccScale=<0.05 0.05 0.1>
    componentLinearVelScale=<0.05 0.05 0.1>
    overrideWorldGravity {z=-980}
   }
   RigidBody() {
    bOverrideWorldGravity=true
    overridePhysicsAsset=&/F_MED_Plague_Dynamics
    simulationSpace=BaseBoneSpace
    alphaBoolBlend {blendOption=Linear}
    baseBoneRef {boneName=pelvis}
    componentAppliedLinearAccClamp=<3000 3000 3000>
    componentLinearAccScale=<0.2 0.2 0.6>
    componentLinearVelScale=<0.2 0.2 0.6>
    overrideWorldGravity {z=-980}
   }
   n13 = "Constraint - dyn_skirt_fr_01_l (R)"!(ComponentPose=Local To Component(LocalPose=Copy Pose From Mesh() {bUseAttachedParent=true}), Alpha=1) -> (Pose) {
    alphaBoolBlend {blendOption=Linear}
    boneToModify {boneName=dyn_skirt_fr_01_l}
    constraintSetup [{transformType=Rotation, targetBone {boneName=thigh_twist_01_l}}]
    constraintWeights [1]
   }
   n14 = "Constraint - dyn_skirt_frdiag_01_l (RR)"!(ComponentPose=@n13, Alpha=1, ConstraintWeights_0=0.5, ConstraintWeights_1:float=1) -> (Pose) {
    alphaBoolBlend {blendOption=Linear}
    boneToModify {boneName=dyn_skirt_frdiag_01_l}
    constraintSetup [{transformType=Rotation, targetBone {boneName=dyn_skirt_fr_01_l}}, {transformType=Rotation, targetBone {boneName=dyn_skirt_out_01_l}}]
    constraintWeights [1,1]
   }
   n15 = "Constraint - dyn_skirt_fr_01_r (R)"!(ComponentPose=@n14, Alpha=1) -> (Pose) {
    alphaBoolBlend {blendOption=Linear}
    boneToModify {boneName=dyn_skirt_fr_01_r}
    constraintSetup [{transformType=Rotation, targetBone {boneName=thigh_twist_01_r}}]
    constraintWeights [1]
   }
   n16 = "Constraint - dyn_skirt_frdiag_01_r (RR)"!(ComponentPose=@n15, Alpha=0, ConstraintWeights_0=0.5, ConstraintWeights_1:float=1) -> (Pose) {
    alphaBoolBlend {blendOption=Linear}
    boneToModify {boneName=dyn_skirt_frdiag_01_r}
    constraintSetup [{transformType=Rotation, targetBone {boneName=dyn_skirt_fr_01_r}}, {transformType=Rotation, targetBone {boneName=dyn_skirt_out_01_r}}]
    constraintWeights [1,1]
   }
   n17 = "Transform (Modify) Bone - Bone: dyn_skirt_fr_01_r"(ComponentPose=@n16, Alpha=Get AimLeft() -> (AimLeft:double), Translation="0.141,0.000,-0.731", Rotation="27.843,0.000,0.000", Scale="1,1,1") {
    rotationMode=BMM_Additive
    rotationSpace=BCS_BoneSpace
    translationMode=BMM_Additive
    translationSpace=BCS_BoneSpace
    alphaBoolBlend {blendOption=Linear}
    alphaScaleBiasClamp {interpSpeedDecreasing=1, interpSpeedIncreasing=1}
    boneToModify {boneName=dyn_skirt_fr_01_r}
    rotation {pitch=27.843}
    translation {x=0.141, z=-0.731}
   }
   n18 = "Transform (Modify) Bone - Bone: dyn_skirt_fr_01_l"(ComponentPose=@n17, Alpha=Get Aim_Right() -> (Aim_Right:double), Translation="0,0,0", Rotation="0,0,0", Scale="1,1,1") {
    rotationMode=BMM_Additive
    rotationSpace=BCS_BoneSpace
    translationMode=BMM_Additive
    translationSpace=BCS_BoneSpace
    alphaBoolBlend {blendOption=Linear}
    alphaScaleBiasClamp {interpSpeedDecreasing=1, interpSpeedIncreasing=1}
    boneToModify {boneName=dyn_skirt_fr_01_l}
   }
   n19 = Get crouching() -> (crouching:double)
   n20 = "dyn_skirt_frdiag_01_l.rotateY = thigh_l.rotateZ * -0.5 - Bone Driven Controller"!(ComponentPose:Component Space Pose Link Structure=@n18, Alpha:float=@n19) -> (Pose:Component Space Pose Link Structure) {
    is=title AnimGraphNode_BoneDrivenController
    bAffectTargetRotationY=true
    multiplier=-0.5
    sourceComponent=RotationZ
    alphaBoolBlend {blendOption=Linear}
    sourceBone {boneName=thigh_l}
    targetBone {boneName=dyn_skirt_frdiag_01_l}
   }
   n21 = "dyn_skirt_frdiag_01_r.rotateY = thigh_r.rotateZ * -0.3 - Bone Driven Controller"!(ComponentPose:Component Space Pose Link Structure=@n20, Alpha:float=@n19) -> (Pose:Component Space Pose Link Structure) {
    is=title AnimGraphNode_BoneDrivenController
    bAffectTargetRotationY=true
    multiplier=-0.3
    sourceComponent=RotationZ
    alphaBoolBlend {blendOption=Linear}
    sourceBone {boneName=thigh_r}
    targetBone {boneName=dyn_skirt_frdiag_01_r}
   }
   n22 = "Transform (Modify) Bone - Bone: dyn_skirt_fr_01_l"(ComponentPose=@n21, Alpha=@n19, Translation="3.326,-1.121,5.052", Rotation="-2.932,-7.713,0.175", Scale="1,1,1") {
    rotationMode=BMM_Additive
    rotationSpace=BCS_BoneSpace
    translationMode=BMM_Additive
    translationSpace=BCS_BoneSpace
    alphaBoolBlend {blendOption=Linear}
    boneToModify {boneName=dyn_skirt_fr_01_l}
    rotation=<rot -2.932 -7.7133 0.175>
    translation=<3.326 -1.121 5.052>
   }
   n23 = Get skydiving vector force() -> (skydiving vector force:Vector)
   n24 = Anim Dynamics - Bone: dyn_skirt_fr_01_l(ComponentPose=@n22, Alpha=1, ExternalForce=@n23) {
    angularDampingOverride=0.9999
    bGravityOverrideInSimSpace=true
    bOverrideAngularDamping=true
    bUseGravityOverride=true
    numSolverIterationsPostUpdate=2
    numSolverIterationsPreUpdate=8
    simulationSpace=RootRelative
    alphaBoolBlend {blendOption=Linear}
    boundBone {boneName=dyn_skirt_fr_01_l}
    componentAppliedLinearAccClamp=<3000 3000 3000>
    componentLinearAccScale=<0.1 0.1 0.1>
    componentLinearVelScale=<0.1 0.1 0.1>
    gravityOverride {z=-980}
    physicsBodyDefinitions [{
      sphereCollisionRadius=0
      boundBone {boneName=dyn_skirt_fr_01_l}
      boxExtents=<20 20 20>
      constraintSetup {twistAxis=AxisZ, angularLimitsMax=<0 0 0>, angularLimitsMin=<0 -30 -10>, angularTarget=<0 0 0>}
      localJointOffset=<3 0 0>
     }]
   }
   n25 = Anim Dynamics - Bone: dyn_skirt_fr_02_l(ComponentPose=@n24, Alpha=1, ExternalForce=@n23) {
    angularDampingOverride=0.9999
    bGravityOverrideInSimSpace=true
    bOverrideAngularDamping=true
    bUseGravityOverride=true
    simulationSpace=RootRelative
    alphaBoolBlend {blendOption=Linear}
    boundBone {boneName=dyn_skirt_fr_02_l}
    componentAppliedLinearAccClamp=<3000 3000 3000>
    componentLinearAccScale=<0.1 0.1 0.1>
    componentLinearVelScale=<0.1 0.1 0.1>
    gravityOverride {z=-980}
    physicsBodyDefinitions [{
      sphereCollisionRadius=0
      boundBone {boneName=dyn_skirt_fr_02_l}
      boxExtents=<20 20 20>
      constraintSetup {twistAxis=AxisZ, angularLimitsMax=<0 5 6>, angularLimitsMin=<0 -15 -6>, angularTarget=<0 0 0>}
      localJointOffset=<3 0 0>
     }]
   }
   n26 = Anim Dynamics - Bone: dyn_skirt_fr_03_l(ComponentPose=@n25, Alpha=1, ExternalForce=@n23) {
    angularDampingOverride=0.9999
    bGravityOverrideInSimSpace=true
    bOverrideAngularDamping=true
    bUseGravityOverride=true
    simulationSpace=RootRelative
    alphaBoolBlend {blendOption=Linear}
    boundBone {boneName=dyn_skirt_fr_03_l}
    componentAppliedLinearAccClamp=<3000 3000 3000>
    componentLinearAccScale=<0.1 0.1 0.1>
    componentLinearVelScale=<0.1 0.1 0.1>
    gravityOverride {z=-980}
    physicsBodyDefinitions [{
      sphereCollisionRadius=0
      boundBone {boneName=dyn_skirt_fr_03_l}
      boxExtents=<20 20 20>
      constraintSetup {twistAxis=AxisZ, angularLimitsMax=<0 15 6>, angularLimitsMin=<0 -15 -6>, angularTarget=<0 0 0>}
      localJointOffset=<3 0 0>
     }]
   }
   n27 = Anim Dynamics - Bone: dyn_skirt_fr_01_r(ComponentPose=@n26, Alpha=1, ExternalForce=@n23) {
    angularDampingOverride=0.9999
    bGravityOverrideInSimSpace=true
    bOverrideAngularDamping=true
    bUseGravityOverride=true
    simulationSpace=RootRelative
    alphaBoolBlend {blendOption=Linear}
    boundBone {boneName=dyn_skirt_fr_01_r}
    componentAppliedLinearAccClamp=<3000 3000 3000>
    componentLinearAccScale=<0.1 0.1 0.1>
    componentLinearVelScale=<0.1 0.1 0.1>
    gravityOverride {z=-980}
    physicsBodyDefinitions [{
      sphereCollisionRadius=0
      boundBone {boneName=dyn_skirt_fr_01_r}
      boxExtents=<20 20 20>
      constraintSetup {twistAxis=AxisZ, angularLimitsMax=<0 0 10>, angularLimitsMin=<0 -30 -10>, angularTarget=<0 0 0>}
      localJointOffset=<-3 0 0>
     }]
   }
   n28 = Anim Dynamics - Bone: dyn_skirt_fr_02_r(ComponentPose=@n27, Alpha=1, ExternalForce=@n23) {
    angularDampingOverride=0.9999
    bGravityOverrideInSimSpace=true
    bOverrideAngularDamping=true
    bUseGravityOverride=true
    bUseSphericalLimits=true
    simulationSpace=RootRelative
    alphaBoolBlend {blendOption=Linear}
    boundBone {boneName=dyn_skirt_fr_02_r}
    componentAppliedLinearAccClamp=<3000 3000 3000>
    componentLinearAccScale=<0.1 0.1 0.1>
    componentLinearVelScale=<0.1 0.1 0.1>
    gravityOverride {z=-980}
    physicsBodyDefinitions [{
      collisionType=CustomSphere
      sphereCollisionRadius=1
      boundBone {boneName=dyn_skirt_fr_02_r}
      boxExtents=<20 20 20>
      constraintSetup {twistAxis=AxisZ, angularLimitsMax=<2 5 6>, angularLimitsMin=<-2 -15 -6>, angularTarget=<0 0 0>}
      localJointOffset=<-3 0 0>
     }]
    sphericalLimits [{limitRadius=9.1, drivingBone {boneName=thigh_r}, sphereLocalOffset=<8 0 0>}]
   }
   n29 = Anim Dynamics - Bone: dyn_skirt_fr_03_r(ComponentPose=@n28, Alpha=1, ExternalForce=@n23) {
    angularDampingOverride=0.9999
    bGravityOverrideInSimSpace=true
    bOverrideAngularDamping=true
    bUseGravityOverride=true
    bUseSphericalLimits=true
    simulationSpace=RootRelative
    alphaBoolBlend {blendOption=Linear}
    boundBone {boneName=dyn_skirt_fr_03_r}
    componentAppliedLinearAccClamp=<3000 3000 3000>
    componentLinearAccScale=<0.1 0.1 0.1>
    componentLinearVelScale=<0.1 0.1 0.1>
    gravityOverride {z=-980}
    physicsBodyDefinitions [{
      collisionType=CustomSphere
      sphereCollisionRadius=1
      boundBone {boneName=dyn_skirt_fr_03_r}
      boxExtents=<20 20 20>
      constraintSetup {twistAxis=AxisZ, angularLimitsMax=<2 10 6>, angularLimitsMin=<-2 -15 -6>, angularTarget=<0 0 0>}
      localJointOffset=<-3 0 0>
     }]
    sphericalLimits [{limitRadius=9.2, drivingBone {boneName=thigh_r}, sphereLocalOffset=<24 0 0>}]
   }
   n30 = "Constraint - dyn_skirt_fr_01 (PP)"!(ComponentPose=@n29, Alpha=1, ConstraintWeights_0=1, ConstraintWeights_1:float=1) -> (Pose) {
    alphaBoolBlend {blendOption=Linear}
    boneToModify {boneName=dyn_skirt_fr_01}
    constraintSetup [{targetBone {boneName=dyn_skirt_fr_01_l}}, {targetBone {boneName=dyn_skirt_fr_01_r}}]
    constraintWeights [1,1]
   }
   n31 = "Constraint - dyn_skirt_fr_02 (PP)"!(ComponentPose=@n30, Alpha=1, ConstraintWeights_0=1, ConstraintWeights_1:float=1) -> (Pose) {
    alphaBoolBlend {blendOption=Linear}
    boneToModify {boneName=dyn_skirt_fr_02}
    constraintSetup [{targetBone {boneName=dyn_skirt_fr_02_l}}, {targetBone {boneName=dyn_skirt_fr_02_r}}]
    constraintWeights [1,1]
   }
   n32 = "Constraint - dyn_skirt_fr_03 (PP)"!(ComponentPose=@n31, Alpha=1, ConstraintWeights_0=1, ConstraintWeights_1:float=1) -> (Pose) {
    alphaBoolBlend {blendOption=Linear}
    boneToModify {boneName=dyn_skirt_fr_03}
    constraintSetup [{targetBone {boneName=dyn_skirt_fr_03_l}}, {targetBone {boneName=dyn_skirt_fr_03_r}}]
    constraintWeights [1,1]
   }
   n33 = Anim Dynamics - Bone: dyn_skirt_fr_01(ComponentPose=@n32, Alpha=1, ExternalForce=@n23) {
    angularDampingOverride=0.9999
    bGravityOverrideInSimSpace=true
    bOverrideAngularDamping=true
    bUseGravityOverride=true
    simulationSpace=RootRelative
    alphaBoolBlend {blendOption=Linear}
    boundBone {boneName=dyn_skirt_fr_01}
    componentAppliedLinearAccClamp=<3000 3000 3000>
    componentLinearAccScale=<0.1 0.1 0.1>
    componentLinearVelScale=<0.1 0.1 0.1>
    gravityOverride {z=-980}
    physicsBodyDefinitions [{
      sphereCollisionRadius=0
      boundBone {boneName=dyn_skirt_fr_01}
      boxExtents=<20 20 20>
      constraintSetup {twistAxis=AxisZ, angularLimitsMax=<0 0 5>, angularLimitsMin=<0 -30 -5>, angularTarget=<0 0 0>}
      localJointOffset=<3 0 0>
     }]
   }
   n34 = Anim Dynamics - Bone: dyn_skirt_fr_02(ComponentPose=@n33, Alpha=1, ExternalForce=@n23) {
    angularDampingOverride=0.9999
    bGravityOverrideInSimSpace=true
    bOverrideAngularDamping=true
    bUseGravityOverride=true
    simulationSpace=RootRelative
    alphaBoolBlend {blendOption=Linear}
    boundBone {boneName=dyn_skirt_fr_02}
    componentAppliedLinearAccClamp=<3000 3000 3000>
    componentLinearAccScale=<0.1 0.1 0.1>
    componentLinearVelScale=<0.1 0.1 0.1>
    gravityOverride {z=-980}
    physicsBodyDefinitions [{
      sphereCollisionRadius=0
      boundBone {boneName=dyn_skirt_fr_02}
      boxExtents=<20 20 20>
      constraintSetup {twistAxis=AxisZ, angularLimitsMax=<10 5 5>, angularLimitsMin=<-10 -15 -5>, angularTarget=<0 0 0>}
      localJointOffset=<3 0 0>
     }]
   }
   n35 = Anim Dynamics - Bone: dyn_skirt_fr_03(ComponentPose=@n34, Alpha=1, ExternalForce=@n23) {
    angularDampingOverride=0.9999
    bGravityOverrideInSimSpace=true
    bOverrideAngularDamping=true
    bUseGravityOverride=true
    simulationSpace=RootRelative
    alphaBoolBlend {blendOption=Linear}
    boundBone {boneName=dyn_skirt_fr_03}
    componentAppliedLinearAccClamp=<3000 3000 3000>
    componentLinearAccScale=<0.1 0.1 0.1>
    componentLinearVelScale=<0.1 0.1 0.1>
    gravityOverride {z=-980}
    physicsBodyDefinitions [{
      sphereCollisionRadius=0
      boundBone {boneName=dyn_skirt_fr_03}
      boxExtents=<20 20 20>
      constraintSetup {twistAxis=AxisZ, angularLimitsMax=<10 15 5>, angularLimitsMin=<-10 -15 -5>, angularTarget=<0 0 0>}
      localJointOffset=<3 0 0>
     }]
   }
   n36 = RigidBody(ComponentPose=@n35, ExternalForce=@n23) {
    bOverrideWorldGravity=true
    overridePhysicsAsset=&/F_MED_Plague_Necklace_Dynamics
    simulationSpace=BaseBoneSpace
    alphaBoolBlend {blendOption=Linear}
    baseBoneRef {boneName=spine_05}
    componentAppliedLinearAccClamp=<3000 3000 3000>
    componentLinearAccScale=<0.2 0.2 0.2>
    componentLinearVelScale=<0.2 0.2 0.2>
    overrideWorldGravity {z=-980}
   }
   n37 = "Transform (Modify) Bone - Bone: dyn_root"(ComponentPose=@n36, Alpha=1, Translation="0,0,0", Rotation=Get sim_space_rotation() -> (sim_space_rotation:Rotator), Scale="1,1,1") {
    rotationMode=BMM_Replace
    rotationSpace=BCS_WorldSpace
    translationSpace=BCS_BoneSpace
    alphaBoolBlend {blendOption=Linear}
    alphaScaleBiasClamp {bInterpResult=true, interpSpeedDecreasing=2, interpSpeedIncreasing=2}
    boneToModify {boneName=dyn_root}
   }
   n38 = RigidBody(ComponentPose=@n37, ExternalForce=@n23) {
    bOverrideWorldGravity=true
    overridePhysicsAsset=&F_MED_Plague_Dynamics
    simulationSpace=BaseBoneSpace
    alphaBoolBlend {blendOption=Linear}
    baseBoneRef {boneName=pelvis}
    componentAppliedLinearAccClamp=<5000 5000 5000>
    componentLinearAccScale=<0.4 0.4 0.6>
    componentLinearVelScale=<0.4 0.4 0.6>
    overrideWorldGravity {z=-980}
    simSpaceSettings {
     maxAngularAcceleration=1000
     maxAngularVelocity=3
     maxLinearAcceleration=700
     maxLinearVelocity=500
     velocityScaleZ=0.2
     worldAlpha=1
     externalLinearDragV=<0.2 0.2>
    }
   }
   n39 = Anim Dynamics - Bone: dyn_forearm_tassle_out_A_01_r(ComponentPose=@n38, Alpha=1, ExternalForce=@n23) {
    angularDampingOverride=0.96
    bGravityOverrideInSimSpace=true
    bOverrideAngularDamping=true
    bUseGravityOverride=true
    numSolverIterationsPostUpdate=2
    numSolverIterationsPreUpdate=8
    simulationSpace=RootRelative
    alphaBoolBlend {blendOption=Linear}
    boundBone {boneName=dyn_forearm_tassle_out_A_01_r}
    componentAppliedLinearAccClamp=<3000 3000 3000>
    componentLinearAccScale=<0.4 0.4 0.4>
    componentLinearVelScale=<0.4 0.4 0.4>
    gravityOverride {z=-980}
    physicsBodyDefinitions [{
      sphereCollisionRadius=0
      boundBone {boneName=dyn_forearm_tassle_out_A_01_r}
      boxExtents=<15 10 15>
      constraintSetup {twistAxis=AxisY, angularLimitsMax=<35 5 35>, angularLimitsMin=<-35 -5 -35>, angularTarget=<0 0 0>}
      localJointOffset=<0 3 0>
     }]
   }
   n40 = Anim Dynamics - Bone: dyn_forearm_tassle_out_A_02_r(ComponentPose=@n39, Alpha=1, ExternalForce=@n23) {
    angularDampingOverride=0.96
    bGravityOverrideInSimSpace=true
    bOverrideAngularDamping=true
    bUseGravityOverride=true
    numSolverIterationsPostUpdate=2
    numSolverIterationsPreUpdate=8
    simulationSpace=RootRelative
    alphaBoolBlend {blendOption=Linear}
    boundBone {boneName=dyn_forearm_tassle_out_A_02_r}
    componentAppliedLinearAccClamp=<3000 3000 3000>
    componentLinearAccScale=<0.5 0.5 0.5>
    componentLinearVelScale=<0.5 0.5 0.5>
    gravityOverride {z=-980}
    physicsBodyDefinitions [{
      sphereCollisionRadius=0
      boundBone {boneName=dyn_forearm_tassle_out_A_02_r}
      boxExtents=<15 10 15>
      constraintSetup {twistAxis=AxisY, angularLimitsMax=<35 5 35>, angularLimitsMin=<-35 -5 -35>, angularTarget=<0 0 0>}
      localJointOffset=<0 3 0>
     }]
   }
   n41 = Anim Dynamics - Bone: dyn_forearm_tassle_out_B_01_r(ComponentPose=@n40, Alpha=1, ExternalForce=@n23) {
    angularDampingOverride=0.96
    bGravityOverrideInSimSpace=true
    bOverrideAngularDamping=true
    bUseGravityOverride=true
    numSolverIterationsPostUpdate=2
    numSolverIterationsPreUpdate=8
    simulationSpace=RootRelative
    alphaBoolBlend {blendOption=Linear}
    boundBone {boneName=dyn_forearm_tassle_out_B_01_r}
    componentAppliedLinearAccClamp=<3000 3000 3000>
    componentLinearAccScale=<0.5 0.5 0.5>
    componentLinearVelScale=<0.5 0.5 0.5>
    gravityOverride {z=-980}
    physicsBodyDefinitions [{
      sphereCollisionRadius=0
      boundBone {boneName=dyn_forearm_tassle_out_B_01_r}
      boxExtents=<15 10 15>
      constraintSetup {twistAxis=AxisY, angularLimitsMax=<35 5 35>, angularLimitsMin=<-35 -5 -35>, angularTarget=<0 0 0>}
      localJointOffset=<0 3 0>
     }]
   }
   n42 = Anim Dynamics - Bone: dyn_forearm_tassle_out_B_02_r(ComponentPose=@n41, Alpha=1, ExternalForce=@n23) {
    angularDampingOverride=0.96
    bGravityOverrideInSimSpace=true
    bOverrideAngularDamping=true
    bUseGravityOverride=true
    numSolverIterationsPostUpdate=2
    numSolverIterationsPreUpdate=8
    simulationSpace=RootRelative
    alphaBoolBlend {blendOption=Linear}
    boundBone {boneName=dyn_forearm_tassle_out_B_02_r}
    componentAppliedLinearAccClamp=<3000 3000 3000>
    componentLinearAccScale=<0.55 0.55 0.55>
    componentLinearVelScale=<0.55 0.55 0.55>
    gravityOverride {z=-980}
    physicsBodyDefinitions [{
      sphereCollisionRadius=0
      boundBone {boneName=dyn_forearm_tassle_out_B_02_r}
      boxExtents=<15 10 15>
      constraintSetup {twistAxis=AxisY, angularLimitsMax=<35 5 35>, angularLimitsMin=<-35 -5 -35>, angularTarget=<0 0 0>}
      localJointOffset=<0 3 0>
     }]
   }
   n43 = Anim Dynamics - Bone: dyn_forearm_tassle_mid_B_01_r(ComponentPose=@n42, Alpha=1, ExternalForce=@n23) {
    angularDampingOverride=0.96
    bGravityOverrideInSimSpace=true
    bOverrideAngularDamping=true
    bUseGravityOverride=true
    numSolverIterationsPostUpdate=2
    numSolverIterationsPreUpdate=8
    simulationSpace=RootRelative
    alphaBoolBlend {blendOption=Linear}
    boundBone {boneName=dyn_forearm_tassle_mid_B_01_r}
    componentAppliedLinearAccClamp=<3000 3000 3000>
    componentLinearAccScale=<0.55 0.55 0.55>
    componentLinearVelScale=<0.55 0.55 0.55>
    gravityOverride {z=-980}
    physicsBodyDefinitions [{
      sphereCollisionRadius=0
      boundBone {boneName=dyn_forearm_tassle_mid_B_01_r}
      boxExtents=<15 10 15>
      constraintSetup {twistAxis=AxisY, angularLimitsMax=<35 5 35>, angularLimitsMin=<-35 -5 -35>, angularTarget=<0 0 0>}
      localJointOffset=<0 3 0>
     }]
   }
   n44 = Anim Dynamics - Bone: dyn_forearm_tassle_mid_B_02_r(ComponentPose=@n43, Alpha=1, ExternalForce=@n23) {
    angularDampingOverride=0.96
    bGravityOverrideInSimSpace=true
    bOverrideAngularDamping=true
    bUseGravityOverride=true
    numSolverIterationsPostUpdate=2
    numSolverIterationsPreUpdate=8
    simulationSpace=RootRelative
    alphaBoolBlend {blendOption=Linear}
    boundBone {boneName=dyn_forearm_tassle_mid_B_02_r}
    componentAppliedLinearAccClamp=<3000 3000 3000>
    componentLinearAccScale=<0.4 0.4 0.4>
    componentLinearVelScale=<0.4 0.4 0.4>
    gravityOverride {z=-980}
    physicsBodyDefinitions [{
      sphereCollisionRadius=0
      boundBone {boneName=dyn_forearm_tassle_mid_B_02_r}
      boxExtents=<15 10 15>
      constraintSetup {twistAxis=AxisY, angularLimitsMax=<35 5 35>, angularLimitsMin=<-35 -5 -35>, angularTarget=<0 0 0>}
      localJointOffset=<0 3 0>
     }]
   }
   n45 = Anim Dynamics - Bone: dyn_forearm_tassle_mid_A_01_r(ComponentPose=@n44, Alpha=1, ExternalForce=@n23) {
    angularDampingOverride=0.96
    bGravityOverrideInSimSpace=true
    bOverrideAngularDamping=true
    bUseGravityOverride=true
    numSolverIterationsPostUpdate=2
    numSolverIterationsPreUpdate=8
    simulationSpace=RootRelative
    alphaBoolBlend {blendOption=Linear}
    boundBone {boneName=dyn_forearm_tassle_mid_A_01_r}
    componentAppliedLinearAccClamp=<3000 3000 3000>
    componentLinearAccScale=<0.5 0.5 0.5>
    componentLinearVelScale=<0.5 0.5 0.5>
    gravityOverride {z=-980}
    physicsBodyDefinitions [{
      sphereCollisionRadius=0
      boundBone {boneName=dyn_forearm_tassle_mid_A_01_r}
      boxExtents=<15 10 15>
      constraintSetup {twistAxis=AxisY, angularLimitsMax=<35 5 35>, angularLimitsMin=<-35 -5 -35>, angularTarget=<0 0 0>}
      localJointOffset=<0 3 0>
     }]
   }
   n46 = Anim Dynamics - Bone: dyn_forearm_tassle_mid_A_02_r(ComponentPose=@n45, Alpha=1, ExternalForce=@n23) {
    angularDampingOverride=0.96
    bGravityOverrideInSimSpace=true
    bOverrideAngularDamping=true
    bUseGravityOverride=true
    numSolverIterationsPostUpdate=2
    numSolverIterationsPreUpdate=8
    simulationSpace=RootRelative
    alphaBoolBlend {blendOption=Linear}
    boundBone {boneName=dyn_forearm_tassle_mid_A_02_r}
    componentAppliedLinearAccClamp=<3000 3000 3000>
    componentLinearAccScale=<0.55 0.55 0.55>
    componentLinearVelScale=<0.55 0.55 0.55>
    gravityOverride {z=-980}
    physicsBodyDefinitions [{
      sphereCollisionRadius=0
      boundBone {boneName=dyn_forearm_tassle_mid_A_02_r}
      boxExtents=<15 10 15>
      constraintSetup {twistAxis=AxisY, angularLimitsMax=<35 5 35>, angularLimitsMin=<-35 -5 -35>, angularTarget=<0 0 0>}
      localJointOffset=<0 3 0>
     }]
   }
   n47 = Anim Dynamics - Bone: dyn_forearm_tassle_up_A_01_r(ComponentPose=@n46, Alpha=1, ExternalForce=@n23) {
    angularDampingOverride=0.96
    bGravityOverrideInSimSpace=true
    bOverrideAngularDamping=true
    bUseGravityOverride=true
    numSolverIterationsPostUpdate=2
    numSolverIterationsPreUpdate=8
    simulationSpace=RootRelative
    alphaBoolBlend {blendOption=Linear}
    boundBone {boneName=dyn_forearm_tassle_up_A_01_r}
    componentAppliedLinearAccClamp=<3000 3000 3000>
    componentLinearAccScale=<0.4 0.4 0.4>
    componentLinearVelScale=<0.4 0.4 0.4>
    gravityOverride {z=-980}
    physicsBodyDefinitions [{
      sphereCollisionRadius=0
      boundBone {boneName=dyn_forearm_tassle_up_A_01_r}
      boxExtents=<15 10 15>
      constraintSetup {twistAxis=AxisY, angularLimitsMax=<35 5 35>, angularLimitsMin=<-35 -5 -35>, angularTarget=<0 0 0>}
      localJointOffset=<0 3 0>
     }]
   }
   n48 = Anim Dynamics - Bone: dyn_forearm_tassle_up_A_02_r(ComponentPose=@n47, Alpha=1, ExternalForce=@n23) {
    angularDampingOverride=0.96
    bGravityOverrideInSimSpace=true
    bOverrideAngularDamping=true
    bUseGravityOverride=true
    numSolverIterationsPostUpdate=2
    numSolverIterationsPreUpdate=8
    simulationSpace=RootRelative
    alphaBoolBlend {blendOption=Linear}
    boundBone {boneName=dyn_forearm_tassle_up_A_02_r}
    componentAppliedLinearAccClamp=<3000 3000 3000>
    componentLinearAccScale=<0.5 0.5 0.5>
    componentLinearVelScale=<0.5 0.5 0.5>
    gravityOverride {z=-980}
    physicsBodyDefinitions [{
      sphereCollisionRadius=0
      boundBone {boneName=dyn_forearm_tassle_up_A_02_r}
      boxExtents=<15 10 15>
      constraintSetup {twistAxis=AxisY, angularLimitsMax=<35 5 35>, angularLimitsMin=<-35 -5 -35>, angularTarget=<0 0 0>}
      localJointOffset=<0 3 0>
     }]
   }
   n49 = Anim Dynamics - Bone: dyn_forearm_tassle_up_B_01_r(ComponentPose=@n48, Alpha=1, ExternalForce=@n23) {
    angularDampingOverride=0.96
    bGravityOverrideInSimSpace=true
    bOverrideAngularDamping=true
    bUseGravityOverride=true
    numSolverIterationsPostUpdate=2
    numSolverIterationsPreUpdate=8
    simulationSpace=RootRelative
    alphaBoolBlend {blendOption=Linear}
    boundBone {boneName=dyn_forearm_tassle_up_B_01_r}
    componentAppliedLinearAccClamp=<3000 3000 3000>
    componentLinearAccScale=<0.55 0.55 0.55>
    componentLinearVelScale=<0.55 0.55 0.55>
    gravityOverride {z=-980}
    physicsBodyDefinitions [{
      sphereCollisionRadius=0
      boundBone {boneName=dyn_forearm_tassle_up_B_01_r}
      boxExtents=<15 10 15>
      constraintSetup {twistAxis=AxisY, angularLimitsMax=<35 5 35>, angularLimitsMin=<-35 -5 -35>, angularTarget=<0 0 0>}
      localJointOffset=<0 3 0>
     }]
   }
   n50 = Anim Dynamics - Bone: dyn_forearm_tassle_up_B_02_r(ComponentPose=@n49, Alpha=1, ExternalForce=@n23) {
    angularDampingOverride=0.96
    bGravityOverrideInSimSpace=true
    bOverrideAngularDamping=true
    bUseGravityOverride=true
    numSolverIterationsPostUpdate=2
    numSolverIterationsPreUpdate=8
    simulationSpace=RootRelative
    alphaBoolBlend {blendOption=Linear}
    boundBone {boneName=dyn_forearm_tassle_up_B_02_r}
    componentAppliedLinearAccClamp=<3000 3000 3000>
    componentLinearAccScale=<0.4 0.4 0.4>
    componentLinearVelScale=<0.4 0.4 0.4>
    gravityOverride {z=-980}
    physicsBodyDefinitions [{
      sphereCollisionRadius=0
      boundBone {boneName=dyn_forearm_tassle_up_B_02_r}
      boxExtents=<15 10 15>
      constraintSetup {twistAxis=AxisY, angularLimitsMax=<35 5 35>, angularLimitsMin=<-35 -5 -35>, angularTarget=<0 0 0>}
      localJointOffset=<0 3 0>
     }]
   }
   n51 = Anim Dynamics - Bone: dyn_forearm_tassle_out_A_01_l(ComponentPose=@n50, Alpha=1, ExternalForce=@n23) {
    angularDampingOverride=0.96
    bGravityOverrideInSimSpace=true
    bOverrideAngularDamping=true
    bUseGravityOverride=true
    numSolverIterationsPostUpdate=2
    numSolverIterationsPreUpdate=8
    simulationSpace=RootRelative
    alphaBoolBlend {blendOption=Linear}
    boundBone {boneName=dyn_forearm_tassle_out_A_01_l}
    componentAppliedLinearAccClamp=<3000 3000 3000>
    componentLinearAccScale=<0.4 0.4 0.4>
    componentLinearVelScale=<0.4 0.4 0.4>
    gravityOverride {z=-980}
    physicsBodyDefinitions [{
      sphereCollisionRadius=0
      boundBone {boneName=dyn_forearm_tassle_out_A_01_l}
      boxExtents=<15 10 15>
      constraintSetup {twistAxis=AxisY, angularLimitsMax=<35 5 35>, angularLimitsMin=<-35 -5 -35>, angularTarget=<0 0 0>}
      localJointOffset=<0 -3 0>
     }]
   }
   n52 = Anim Dynamics - Bone: dyn_forearm_tassle_out_A_02_l(ComponentPose=@n51, Alpha=1, ExternalForce=@n23) {
    angularDampingOverride=0.96
    bGravityOverrideInSimSpace=true
    bOverrideAngularDamping=true
    bUseGravityOverride=true
    numSolverIterationsPostUpdate=2
    numSolverIterationsPreUpdate=8
    simulationSpace=RootRelative
    alphaBoolBlend {blendOption=Linear}
    boundBone {boneName=dyn_forearm_tassle_out_A_02_l}
    componentAppliedLinearAccClamp=<3000 3000 3000>
    componentLinearAccScale=<0.5 0.5 0.5>
    componentLinearVelScale=<0.5 0.5 0.5>
    gravityOverride {z=-980}
    physicsBodyDefinitions [{
      sphereCollisionRadius=0
      boundBone {boneName=dyn_forearm_tassle_out_A_02_l}
      boxExtents=<15 10 15>
      constraintSetup {twistAxis=AxisY, angularLimitsMax=<35 5 35>, angularLimitsMin=<-35 -5 -35>, angularTarget=<0 0 0>}
      localJointOffset=<0 -3 0>
     }]
   }
   n53 = Anim Dynamics - Bone: dyn_forearm_tassle_out_B_01_l(ComponentPose=@n52, Alpha=1, ExternalForce=@n23) {
    angularDampingOverride=0.96
    bGravityOverrideInSimSpace=true
    bOverrideAngularDamping=true
    bUseGravityOverride=true
    numSolverIterationsPostUpdate=2
    numSolverIterationsPreUpdate=8
    simulationSpace=RootRelative
    alphaBoolBlend {blendOption=Linear}
    boundBone {boneName=dyn_forearm_tassle_out_B_01_l}
    componentAppliedLinearAccClamp=<3000 3000 3000>
    componentLinearAccScale=<0.55 0.55 0.55>
    componentLinearVelScale=<0.55 0.55 0.55>
    gravityOverride {z=-980}
    physicsBodyDefinitions [{
      sphereCollisionRadius=0
      boundBone {boneName=dyn_forearm_tassle_out_B_01_l}
      boxExtents=<15 10 15>
      constraintSetup {twistAxis=AxisY, angularLimitsMax=<35 5 35>, angularLimitsMin=<-35 -5 -35>, angularTarget=<0 0 0>}
      localJointOffset=<0 -3 0>
     }]
   }
   n54 = Anim Dynamics - Bone: dyn_forearm_tassle_out_B_02_l(ComponentPose=@n53, Alpha=1, ExternalForce=@n23) {
    angularDampingOverride=0.96
    bGravityOverrideInSimSpace=true
    bOverrideAngularDamping=true
    bUseGravityOverride=true
    numSolverIterationsPostUpdate=2
    numSolverIterationsPreUpdate=8
    simulationSpace=RootRelative
    alphaBoolBlend {blendOption=Linear}
    boundBone {boneName=dyn_forearm_tassle_out_B_02_l}
    componentAppliedLinearAccClamp=<3000 3000 3000>
    componentLinearAccScale=<0.5 0.5 0.5>
    componentLinearVelScale=<0.5 0.5 0.5>
    gravityOverride {z=-980}
    physicsBodyDefinitions [{
      sphereCollisionRadius=0
      boundBone {boneName=dyn_forearm_tassle_out_B_02_l}
      boxExtents=<15 10 15>
      constraintSetup {twistAxis=AxisY, angularLimitsMax=<35 5 35>, angularLimitsMin=<-35 -5 -35>, angularTarget=<0 0 0>}
      localJointOffset=<0 -3 0>
     }]
   }
   n55 = Anim Dynamics - Bone: dyn_forearm_tassle_mid_B_01_l(ComponentPose=@n54, Alpha=1, ExternalForce=@n23) {
    angularDampingOverride=0.96
    bGravityOverrideInSimSpace=true
    bOverrideAngularDamping=true
    bUseGravityOverride=true
    numSolverIterationsPostUpdate=2
    numSolverIterationsPreUpdate=8
    simulationSpace=RootRelative
    alphaBoolBlend {blendOption=Linear}
    boundBone {boneName=dyn_forearm_tassle_mid_B_01_l}
    componentAppliedLinearAccClamp=<3000 3000 3000>
    componentLinearAccScale=<0.5 0.5 0.5>
    componentLinearVelScale=<0.5 0.5 0.5>
    gravityOverride {z=-980}
    physicsBodyDefinitions [{
      sphereCollisionRadius=0
      boundBone {boneName=dyn_forearm_tassle_mid_B_01_l}
      boxExtents=<15 10 15>
      constraintSetup {twistAxis=AxisY, angularLimitsMax=<35 5 35>, angularLimitsMin=<-35 -5 -35>, angularTarget=<0 0 0>}
      localJointOffset=<0 -3 0>
     }]
   }
   n56 = Anim Dynamics - Bone: dyn_forearm_tassle_mid_B_02_l(ComponentPose=@n55, Alpha=1, ExternalForce=@n23) {
    angularDampingOverride=0.96
    bGravityOverrideInSimSpace=true
    bOverrideAngularDamping=true
    bUseGravityOverride=true
    numSolverIterationsPostUpdate=2
    numSolverIterationsPreUpdate=8
    simulationSpace=RootRelative
    alphaBoolBlend {blendOption=Linear}
    boundBone {boneName=dyn_forearm_tassle_mid_B_02_l}
    componentAppliedLinearAccClamp=<3000 3000 3000>
    componentLinearAccScale=<0.4 0.4 0.4>
    componentLinearVelScale=<0.4 0.4 0.4>
    gravityOverride {z=-980}
    physicsBodyDefinitions [{
      sphereCollisionRadius=0
      boundBone {boneName=dyn_forearm_tassle_mid_B_02_l}
      boxExtents=<15 10 15>
      constraintSetup {twistAxis=AxisY, angularLimitsMax=<35 5 35>, angularLimitsMin=<-35 -5 -35>, angularTarget=<0 0 0>}
      localJointOffset=<0 -3 0>
     }]
   }
   n57 = Anim Dynamics - Bone: dyn_forearm_tassle_mid_A_01_l(ComponentPose=@n56, Alpha=1, ExternalForce=@n23) {
    angularDampingOverride=0.96
    bGravityOverrideInSimSpace=true
    bOverrideAngularDamping=true
    bUseGravityOverride=true
    numSolverIterationsPostUpdate=2
    numSolverIterationsPreUpdate=8
    simulationSpace=RootRelative
    alphaBoolBlend {blendOption=Linear}
    boundBone {boneName=dyn_forearm_tassle_mid_A_01_l}
    componentAppliedLinearAccClamp=<3000 3000 3000>
    componentLinearAccScale=<0.55 0.55 0.55>
    componentLinearVelScale=<0.55 0.55 0.55>
    gravityOverride {z=-980}
    physicsBodyDefinitions [{
      sphereCollisionRadius=0
      boundBone {boneName=dyn_forearm_tassle_mid_A_01_l}
      boxExtents=<15 10 15>
      constraintSetup {twistAxis=AxisY, angularLimitsMax=<35 5 35>, angularLimitsMin=<-35 -5 -35>, angularTarget=<0 0 0>}
      localJointOffset=<0 -3 0>
     }]
   }
   n58 = Anim Dynamics - Bone: dyn_forearm_tassle_mid_A_02_l(ComponentPose=@n57, Alpha=1, ExternalForce=@n23) {
    angularDampingOverride=0.96
    bGravityOverrideInSimSpace=true
    bOverrideAngularDamping=true
    bUseGravityOverride=true
    numSolverIterationsPostUpdate=2
    numSolverIterationsPreUpdate=8
    simulationSpace=RootRelative
    alphaBoolBlend {blendOption=Linear}
    boundBone {boneName=dyn_forearm_tassle_mid_A_02_l}
    componentAppliedLinearAccClamp=<3000 3000 3000>
    componentLinearAccScale=<0.4 0.4 0.4>
    componentLinearVelScale=<0.4 0.4 0.4>
    gravityOverride {z=-980}
    physicsBodyDefinitions [{
      sphereCollisionRadius=0
      boundBone {boneName=dyn_forearm_tassle_mid_A_02_l}
      boxExtents=<15 10 15>
      constraintSetup {twistAxis=AxisY, angularLimitsMax=<35 5 35>, angularLimitsMin=<-35 -5 -35>, angularTarget=<0 0 0>}
      localJointOffset=<0 -3 0>
     }]
   }
   n59 = Anim Dynamics - Bone: dyn_forearm_tassle_up_A_01_l(ComponentPose=@n58, Alpha=1, ExternalForce=@n23) {
    angularDampingOverride=0.96
    bGravityOverrideInSimSpace=true
    bOverrideAngularDamping=true
    bUseGravityOverride=true
    numSolverIterationsPostUpdate=2
    numSolverIterationsPreUpdate=8
    simulationSpace=RootRelative
    alphaBoolBlend {blendOption=Linear}
    boundBone {boneName=dyn_forearm_tassle_up_A_01_l}
    componentAppliedLinearAccClamp=<3000 3000 3000>
    componentLinearAccScale=<0.4 0.4 0.4>
    componentLinearVelScale=<0.4 0.4 0.4>
    gravityOverride {z=-980}
    physicsBodyDefinitions [{
      sphereCollisionRadius=0
      boundBone {boneName=dyn_forearm_tassle_up_A_01_l}
      boxExtents=<15 10 15>
      constraintSetup {twistAxis=AxisY, angularLimitsMax=<35 5 35>, angularLimitsMin=<-35 -5 -35>, angularTarget=<0 0 0>}
      localJointOffset=<0 -3 0>
     }]
   }
   n60 = Anim Dynamics - Bone: dyn_forearm_tassle_up_A_02_l(ComponentPose=@n59, Alpha=1, ExternalForce=@n23) {
    angularDampingOverride=0.96
    bGravityOverrideInSimSpace=true
    bOverrideAngularDamping=true
    bUseGravityOverride=true
    numSolverIterationsPostUpdate=2
    numSolverIterationsPreUpdate=8
    simulationSpace=RootRelative
    alphaBoolBlend {blendOption=Linear}
    boundBone {boneName=dyn_forearm_tassle_up_A_02_l}
    componentAppliedLinearAccClamp=<3000 3000 3000>
    componentLinearAccScale=<0.55 0.55 0.55>
    componentLinearVelScale=<0.55 0.55 0.55>
    gravityOverride {z=-980}
    physicsBodyDefinitions [{
      sphereCollisionRadius=0
      boundBone {boneName=dyn_forearm_tassle_up_A_02_l}
      boxExtents=<15 10 15>
      constraintSetup {twistAxis=AxisY, angularLimitsMax=<35 5 35>, angularLimitsMin=<-35 -5 -35>, angularTarget=<0 0 0>}
      localJointOffset=<0 -3 0>
     }]
   }
   n61 = Anim Dynamics - Bone: dyn_forearm_tassle_up_B_01_l(ComponentPose=@n60, Alpha=1, ExternalForce=@n23) {
    angularDampingOverride=0.96
    bGravityOverrideInSimSpace=true
    bOverrideAngularDamping=true
    bUseGravityOverride=true
    numSolverIterationsPostUpdate=2
    numSolverIterationsPreUpdate=8
    simulationSpace=RootRelative
    alphaBoolBlend {blendOption=Linear}
    boundBone {boneName=dyn_forearm_tassle_up_B_01_l}
    componentAppliedLinearAccClamp=<3000 3000 3000>
    componentLinearAccScale=<0.4 0.4 0.4>
    componentLinearVelScale=<0.4 0.4 0.4>
    gravityOverride {z=-980}
    physicsBodyDefinitions [{
      sphereCollisionRadius=0
      boundBone {boneName=dyn_forearm_tassle_up_B_01_l}
      boxExtents=<15 10 15>
      constraintSetup {twistAxis=AxisY, angularLimitsMax=<35 5 35>, angularLimitsMin=<-35 -5 -35>, angularTarget=<0 0 0>}
      localJointOffset=<0 -3 0>
     }]
   }
   n62 = Anim Dynamics - Bone: dyn_forearm_tassle_up_B_02_l(ComponentPose=@n61, Alpha=1, ExternalForce=@n23) {
    angularDampingOverride=0.96
    bGravityOverrideInSimSpace=true
    bOverrideAngularDamping=true
    bUseGravityOverride=true
    numSolverIterationsPostUpdate=2
    numSolverIterationsPreUpdate=8
    simulationSpace=RootRelative
    alphaBoolBlend {blendOption=Linear}
    boundBone {boneName=dyn_forearm_tassle_up_B_02_l}
    componentAppliedLinearAccClamp=<3000 3000 3000>
    componentLinearAccScale=<0.5 0.5 0.5>
    componentLinearVelScale=<0.5 0.5 0.5>
    gravityOverride {z=-980}
    physicsBodyDefinitions [{
      sphereCollisionRadius=0
      boundBone {boneName=dyn_forearm_tassle_up_B_02_l}
      boxExtents=<15 10 15>
      constraintSetup {twistAxis=AxisY, angularLimitsMax=<35 5 35>, angularLimitsMin=<-35 -5 -35>, angularTarget=<0 0 0>}
      localJointOffset=<0 -3 0>
     }]
   }
   n63 = Get lod_1_adjust() -> (lod_1_adjust:double)
   n64 = "Transform (Modify) Bone - Bone: dyn_skirt_bk_01_r"(ComponentPose=@n62, Alpha=@n63, Translation="0,0,-5", Rotation="0,0,0", Scale="1,1,1") {
    rotationMode=BMM_Additive
    rotationSpace=BCS_BoneSpace
    translationMode=BMM_Additive
    translationSpace=BCS_BoneSpace
    alphaBoolBlend {blendOption=Linear}
    boneToModify {boneName=dyn_skirt_bk_01_r}
    translation {z=-5}
   }
   n65 = "Transform (Modify) Bone - Bone: dyn_skirt_bk_01_l"(ComponentPose=@n64, Alpha=@n63, Translation="0,0,-5", Rotation="0,0,0", Scale="1,1,1") {
    rotationMode=BMM_Additive
    rotationSpace=BCS_BoneSpace
    translationMode=BMM_Additive
    translationSpace=BCS_BoneSpace
    alphaBoolBlend {blendOption=Linear}
    boneToModify {boneName=dyn_skirt_bk_01_l}
    translation {z=-5}
   }
   n66 = "Transform (Modify) Bone - Bone: dyn_skirt_bk_01"(ComponentPose=@n65, Alpha=@n63, Translation="0,0,5", Rotation="0,0,0", Scale="1,1,1") {
    rotationMode=BMM_Additive
    rotationSpace=BCS_BoneSpace
    translationMode=BMM_Additive
    translationSpace=BCS_BoneSpace
    alphaBoolBlend {blendOption=Linear}
    boneToModify {boneName=dyn_skirt_bk_01}
    translation {z=5}
   }
   n67 = "Transform (Modify) Bone - Bone: dyn_skirt_bkmid_01_r"(ComponentPose=@n66, Alpha=@n63, Translation="0,0,-5", Rotation="0,0,0", Scale="1,1,1") {
    rotationMode=BMM_Additive
    rotationSpace=BCS_BoneSpace
    translationMode=BMM_Additive
    translationSpace=BCS_BoneSpace
    alphaBoolBlend {blendOption=Linear}
    boneToModify {boneName=dyn_skirt_bkmid_01_r}
    translation {z=-5}
   }
   n68 = "Transform (Modify) Bone - Bone: dyn_skirt_bkmid_01_l"(ComponentPose=@n67, Alpha=@n63, Translation="0,0,5", Rotation="0,0,0", Scale="1,1,1") {
    rotationMode=BMM_Additive
    rotationSpace=BCS_BoneSpace
    translationMode=BMM_Additive
    translationSpace=BCS_BoneSpace
    alphaBoolBlend {blendOption=Linear}
    boneToModify {boneName=dyn_skirt_bkmid_01_l}
    translation {z=5}
   }
   n69 = "Transform (Modify) Bone - Bone: dyn_skirt_bkdiag_01_r"(ComponentPose=@n68, Alpha=@n63, Translation="0,0,-4", Rotation="0,0,0", Scale="1,1,1") {
    rotationMode=BMM_Additive
    rotationSpace=BCS_BoneSpace
    translationMode=BMM_Additive
    translationSpace=BCS_BoneSpace
    alphaBoolBlend {blendOption=Linear}
    boneToModify {boneName=dyn_skirt_bkdiag_01_r}
    translation {z=-4}
   }
   n70 = "Transform (Modify) Bone - Bone: dyn_skirt_bkdiag_01_l"(ComponentPose=@n69, Alpha=@n63, Translation="0,0,4", Rotation="0,0,0", Scale="1,1,1") {
    rotationMode=BMM_Additive
    rotationSpace=BCS_BoneSpace
    translationMode=BMM_Additive
    translationSpace=BCS_BoneSpace
    alphaBoolBlend {blendOption=Linear}
    boneToModify {boneName=dyn_skirt_bkdiag_01_l}
    translation {z=4}
   }
   Output Pose(Result=Component To Local(ComponentPose=@n70)) {layerGroup=None}
   Flap Node - Bone: dyn_skirt_fr_01_r!(ComponentPose:Component Space Pose Link Structure, Alpha:float=1) -> (Pose:Component Space Pose Link Structure) {
    is=title FortAnimGraphNode_Flap
    bUseClamp=true
    targetClampMax=120
    targetComponent=RotationZ
    alphaBoolBlend {blendOption=Linear}
    sourceDrives [{
      bUseRange=true
      drivingCurve=null
      multiplier=1
      rangeMax=1
      rangeMin=-1
      remappedMax=1
      remappedMin=0
      sourceComponent=RotationZ
      useQuaternion=false
      sourceBone {boneName=thigh_r}
     }]
    targetBone {boneName=dyn_skirt_fr_01_r}
   }
  ]}
 EventGraph {nodes [
   Event BlueprintUpdateAnimation()
   > n1 = Sequence!(execute) -> (then_0, then_1, then_2:Exec)
   > Set crouching(crouching:double="To Float (Boolean)"(InBool=Get bIsCrouching())) -> (Output_Get:double)
   n2 = GetOwnerSkeletalMeshComponent()
   n3 = BreakRotator(InRot=BreakTransform(InTransform=GetDeltaTransformFromRefPose(self=@n2, BoneName=spine_05, BaseName=pelvis)).Rotation)
   > Set AimLeft(AimLeft:double=MapRangeClamped(Value=@n3, InRangeA=0, InRangeB=22)) -> (Output_Get:double)
   > Set Aim_Right(Aim_Right:double=MapRangeClamped(Value=@n3, InRangeA=0, InRangeB=-22)) -> (Output_Get:double)
   > n4 = Branch(Condition=Get bCanPlayCustomAnimations() -> (bCanPlayCustomAnimations:bool))
   n5 = Branch(execute=[@n4.else, @n4.then], Condition=Get bIsSkyDiving())
   > n6 = Branch(Condition=Get bIsSkydiveFloating())
   n7 = Get Direction Between Sockets!(
    Component:Skeletal Mesh Component Object Reference=GetOwnerSkeletalMeshComponent()
    SocketOrBoneNameFrom:Name=head
    SocketOrBoneNameTo:Name=pelvis
   ) -> (ReturnValue:Vector) {is=title K2Node_CallFunction}
   > Set skydiving vector force(skydiving vector force:Vector=vector + vector(
     A:Vector=vector * vector(A:Vector=@n7, B:Vector="75,75,75") -> (ReturnValue:Vector) {is=title K2Node_CommutativeAssociativeBinaryOperator}
     B:Vector="0, 0,2000"
    ) -> (ReturnValue:Vector) {is=title K2Node_CommutativeAssociativeBinaryOperator}) -> (Output_Get:Vector)
   n8 = Branch(execute=@n6.else, Condition=Get bIsSkydiveDiveMode())
   > Set skydiving vector force(skydiving vector force=vector + vector(A=vector * vector(A=@n7, B="180,180,200"), B="0, 0,2000"))
   Set skydiving vector force(execute=@n8.else, skydiving vector force=vector * vector(A=@n7, B="-5,-5,-5"))
   Set skydiving vector force(execute=@n5.else, skydiving vector force="0, 0, 0")
   n9 = "Equal (Float)"(A="To Float (Integer)"(InInt=Get CurrentLOD()), B=1)
   n10 = Get bRigidBodyEnabled()
   n11 = Get bIsCrouching()
   Set lod_1_adjust(execute=@n1.then_1, lod_1_adjust:double=SelectFloat(A=1, B=0, bPickA=AND Boolean!(A=@n10, B=@n9, C:bool=@n11) -> (ReturnValue))) -> (Output_Get:double)
   n12 = ComplexSimSpaceFunction DEPRECATED!(
    execute=@n1.then_2
    CharacterPartAnimInstance:Custom Character Part Anim Instance Object Reference=Self-Reference()
    InterpSpeedMax:double=9
    PawnSpeedMax:double=20
    RInterpErrorTolerance:double=1
    DeltaTime:double=0
    DebugDraw:bool=false
    sim_space_rotation_in:Rotator=Get sim_space_rotation()
    inComponentSpaceIn:bool=Get InComponentSpace() -> (InComponentSpace:bool)
    inComponentSpaceLastFrameIn:bool=Get InComponentSpaceLastFrame() -> (InComponentSpaceLastFrame:bool)
    lockedOnSimSpaceIn:bool=Get LockedOnSimSpace() -> (LockedOnSimSpace:bool)
    InterpSpeedIn:double=Get InterpSpeed() -> (InterpSpeed:double)
   ) -> (
    then
    sim_space_rotation:Rotator
    inComponentSpaceOut:bool
    inComponentSpaceLastFrameOut:bool
    lockedOnSimSpaceOut:bool
    InterpSpeedOut:double
   )
   > Set sim_space_rotation(sim_space_rotation:Rotator=@n12) -> (Output_Get:Rotator)
   > Set InComponentSpace(InComponentSpace:bool=@n12.inComponentSpaceOut) -> (Output_Get:bool)
   > Set InComponentSpaceLastFrame(InComponentSpaceLastFrame:bool=@n12.inComponentSpaceLastFrameOut) -> (Output_Get:bool)
   > Set LockedOnSimSpace(LockedOnSimSpace:bool=@n12.lockedOnSimSpaceOut) -> (Output_Get:bool)
   > Set InterpSpeed(InterpSpeed:double=@n12.InterpSpeedOut) -> (Output_Get:double)
  ]}
]
variables [
 Aim_Right {type=number}
 AimLeft {type=number}
 crouching {type=number}
 InComponentSpace {type=boolean}
 InComponentSpaceLastFrame {type=boolean}
 InterpSpeed {type=number}
 LockedOnSimSpace {type=boolean}
 lod_1_adjust {type=number}
 sim_space_rotation {type=Rotator}
 skydiving vector force {type=Vector}
]
