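# Physics asset of collision shapes for the female medium Plague player body, evidently for its cape simulation, as the name indicates.
# Capsules cover head, clavicles, pelvis, spine and feet, with perpendicular pairs on spine_02 to spine_04.
# More capsules ride on dynamic skirt and forearm tassel bones, so colliders follow those secondary parts.
# Arm, thigh, calf and neck bodies have no shapes, and no constraints join the bodies.
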PhysicsAsset /BRCosmetics/Characters/Player/Female/Medium/Bodies/F_MED_Plague/Meshes/F_MED_Plague_CapeColliders
bodies {
 calf_l []
 calf_r []
 clavicle_l [{shapeType=Capsule, length=5.4464, radius=8.6353, center=<8.1406 -0.4522 -2.6475>, rotation=<rot -77.5566 177.3915 7.7314>}]
 clavicle_r [{shapeType=Capsule, length=6.6964, radius=9.2957, center=<-8.1406 0.4522 2.6475>, rotation=<rot -77.5565 177.3915 -172.2685>}]
 dyn_forearm_tassle_mid_A_02_l [{
   shapeType=Capsule
   length=7.5
   radius=3
   shapeName=dyn_forearm_tassle_mid_A_02_l_capsule
   center=<-0.2849 -0.0007 -0.4979>
   rotation=<rot -29.4985 -8.4905 94.2744>
  }]
 dyn_forearm_tassle_mid_A_02_r [{
   shapeType=Capsule
   length=7.5
   radius=3
   shapeName=dyn_forearm_tassle_mid_A_02_r_capsule
   center=<0.2808 0.0007 0.4907>
   rotation=<rot -29.4985 -8.4905 94.2744>
  }]
 dyn_forearm_tassle_out_B_02_l [{
   shapeType=Capsule
   length=7.5
   radius=3
   shapeName=dyn_forearm_tassle_out_B_02_l_capsule
   center=<0.1982 0.0005 0.3463>
   rotation=<rot -29.4986 -8.4905 94.2744>
  }]
 dyn_forearm_tassle_out_B_02_r [{
   shapeType=Capsule
   length=7.5
   radius=3
   shapeName=dyn_forearm_tassle_out_B_02_r_capsule
   center=<-0.1371 -0.0003 -0.2395>
   rotation=<rot -29.4986 -8.4905 94.2744>
  }]
 dyn_forearm_tassle_up_B_02_l [{
   shapeType=Capsule
   length=7.5
   radius=3
   shapeName=dyn_forearm_tassle_up_B_02_l_capsule
   center=<0.1563 0.0004 0.2732>
   rotation=<rot -29.4987 -8.4905 94.2744>
  }]
 dyn_forearm_tassle_up_B_02_r [{
   shapeType=Capsule
   length=7.5
   radius=3
   shapeName=dyn_forearm_tassle_up_B_02_r_capsule
   center=<-0.3614 -0.0009 -0.6316>
   rotation=<rot -29.4987 -8.4905 94.2744>
  }]
 dyn_skirt_bk_03 [{shapeType=Capsule, length=31, radius=12.75, shapeName=dyn_skirt_bk_03_capsule, center=<-8.3983 0 -3.7724>, rotation=<rot 90 -180 -180>}]
 dyn_skirt_bk_03_l [{
   shapeType=Capsule
   length=28.75
   radius=10.5
   shapeName=dyn_skirt_bk_03_l_capsule
   center=<-9.6989 -0.0001 -1.5463>
   rotation=<rot 90 -0.0004 -0.0008>
  }]
 dyn_skirt_bkdiag_03_r [{
   shapeType=Capsule
   length=28.75
   radius=10.5
   shapeName=dyn_skirt_bkdiag_03_r_capsule
   center=<8.1413 0.0001 1.1053>
   rotation=<rot 90 -5.0513 -5.0518>
  }]
 dyn_skirt_out_03_l [{
   shapeType=Capsule
   length=29.5
   radius=11.5
   shapeName=dyn_skirt_out_03_l_capsule
   center=<-8.1233 -0.0001 0>
   rotation=<rot 90 -165.1488 -165.1495>
  }]
 dyn_skirt_out_03_r [{shapeType=Capsule, length=31.75, radius=11.5, shapeName=dyn_skirt_out_03_r_capsule, center=<5.8683 0.0001 0>, rotation=<rot 90 -90 -90.0007>}]
 foot_l [{shapeType=Capsule, length=12.5982, radius=6.5127, center=<-2.2072 6.0534 -0.3367>, rotation=<rot 0.0388 1.4701 -89.3982>}]
 foot_r [{shapeType=Capsule, length=12.5982, radius=6.5127, center=<2.2072 -6.0534 0.3367>, rotation=<rot 0.0388 1.4701 90.6018>}]
 head [{shapeType=Capsule, length=0, radius=18, center=<9.0697 0.0746 0>, rotation=<rot 0 -90 0>}]
 lowerarm_l []
 lowerarm_r []
 neck_02 []
 pelvis [{shapeType=Capsule, length=7.4028, radius=15.1596, center=<-5.209 -4.7193 -0.1981>}]
 spine_01 [{shapeType=Capsule, length=8.2671, radius=11.4854, center=<-2.0572 -1.8154 0>}]
 spine_02 [
  {shapeType=Capsule, length=4.2984, radius=9.9854, center=<1.179 2.2974 0>}
  {shapeType=Capsule, length=20.6875, radius=9.3125, center=<-1.8317 2.606 0>, rotation=<rot 90 0 0>}
 ]
 spine_03 [
  {shapeType=Capsule, length=3, radius=11.9854, center=<-0.0747 1.4408 0>}
  {shapeType=Capsule, length=20.7188, radius=9.2812, center=<-1.2326 2.6059 0>, rotation=<rot 90 0 0>}
 ]
 spine_04 [
  {shapeType=Capsule, length=9.0796, radius=11.6416, center=<3.9605 0.6826 0>}
  {shapeType=Capsule, length=10.0312, radius=9.2812, center=<-1.2326 1.308 0>, rotation=<rot 90 0 0>}
 ]
 spine_05 [{shapeType=Capsule, length=10.9859, radius=10.0791, center=<5.6377 0.3227 0>}]
 thigh_l []
 thigh_r []
 upperarm_l []
 upperarm_r []
}
constraints []
