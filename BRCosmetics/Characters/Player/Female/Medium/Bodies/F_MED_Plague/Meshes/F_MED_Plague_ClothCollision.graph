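# Physics asset that supplies cloth collision volumes for the lower body of the female medium Plague player cosmetic.
# Single capsules cover the pelvis, calves and feet, mirrored left to right, while both thigh bodies have no shapes.
# Constraints chain each foot to calf, thigh and pelvis with free swing and twist, so their angular limit values are not enforced.
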PhysicsAsset /BRCosmetics/Characters/Player/Female/Medium/Bodies/F_MED_Plague/Meshes/F_MED_Plague_ClothCollision
bodies {
 calf_l [{shapeType=Capsule, length=11.3678, radius=6.6239, center=<-8.0172 -2.2734 0.058>, rotation=<rot 0 104.762 -84.0787>}]
 calf_r [{shapeType=Capsule, length=11.3678, radius=6.6239, center=<8.0172 2.2734 -0.058>, rotation=<rot 0 104.762 -84.0787>}]
 foot_l [{shapeType=Capsule, length=17.6506, radius=7.0978, center=<-3.225 6.8782 -0.3711>, rotation=<rot -0.4263 0.5705 -84.0192>}]
 foot_r [{shapeType=Capsule, length=17.6506, radius=7.0978, center=<3.225 -6.8782 0.3711>, rotation=<rot -0.4263 0.5705 -84.0192>}]
 pelvis [{shapeType=Capsule, length=8.8684, radius=8.7167, center=<-4.117 -1.8901 0>}]
 thigh_l []
 thigh_r []
}
constraints [
 {
  bone1Name=calf_r
  bone2Name=thigh_r
  swing1LimitDegrees=45
  swing1Motion=Free
  swing2LimitDegrees=45
  swing2Motion=Free
  twistLimitDegrees=45
  twistMotion=Free
 }
 {
  bone1Name=calf_l
  bone2Name=thigh_l
  swing1LimitDegrees=45
  swing1Motion=Free
  swing2LimitDegrees=45
  swing2Motion=Free
  twistLimitDegrees=45
  twistMotion=Free
 }
 {
  bone1Name=foot_l
  bone2Name=calf_l
  swing1LimitDegrees=45
  swing1Motion=Free
  swing2LimitDegrees=45
  swing2Motion=Free
  twistLimitDegrees=45
  twistMotion=Free
 }
 {
  bone1Name=foot_r
  bone2Name=calf_r
  swing1LimitDegrees=45
  swing1Motion=Free
  swing2LimitDegrees=45
  swing2Motion=Free
  twistLimitDegrees=45
  twistMotion=Free
 }
 {
  bone1Name=thigh_l
  bone2Name=pelvis
  swing1LimitDegrees=45
  swing1Motion=Free
  swing2LimitDegrees=45
  swing2Motion=Free
  twistLimitDegrees=45
  twistMotion=Free
 }
 {
  bone1Name=thigh_r
  bone2Name=pelvis
  swing1LimitDegrees=45
  swing1Motion=Free
  swing2LimitDegrees=45
  swing2Motion=Free
  twistLimitDegrees=45
  twistMotion=Free
 }
]
